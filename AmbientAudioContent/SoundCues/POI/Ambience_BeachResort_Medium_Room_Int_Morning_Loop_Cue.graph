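# Sound cue for a looping indoor morning ambience bed in a medium room at the Beach Resort POI.
# A quality-level switch sends its first two slots to a quad loop and its last two to a stereo loop, each through a single-input mixer, both reusing Residential interior waves.
# It plays non-spatialized, without attenuation or reverb send, in the indoor ambient sound class, with a concurrency override that prevents new instances at the limit.
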SoundCue /AmbientAudioContent/SoundCues/POI/Ambience_BeachResort_Medium_Room_Int_Morning_Loop_Cue
root=QualityLevel_0
nodes [
 {id=Mixer_1, children [WavePlayer_5], properties {inputVolume [1]}}
 {id=Mixer_3, children [WavePlayer_4], properties {inputVolume [1]}}
 {id=QualityLevel_0, children [Mixer_1,Mixer_1,Mixer_3,Mixer_3]}
 {id=WavePlayer_4, properties {bLooping=true, soundWaveAssetPtr=/AmbientAudioContent/Waves/POI/Ambience_Residential_Int_Stereo_3}}
 {id=WavePlayer_5, properties {bLooping=true, soundWaveAssetPtr=/AmbientAudioContent/Waves/POI/Ambience_Residential_Int_Quad_3}}
]
properties {
 bOverrideConcurrency=true
 maxDistance=2097152
 soundClassObject=&Fort_Ambient_Indoors_Athena
 volumeMultiplier=1.403
 attenuationOverrides {bAttenuate=false, bEnableReverbSend=false, bSpatialize=false}
 concurrencyOverrides {resolutionRule=PreventNew}
}
