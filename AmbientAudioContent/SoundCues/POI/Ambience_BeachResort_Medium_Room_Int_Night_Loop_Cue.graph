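# Looping night ambience cue for a medium-sized interior room at the beach resort POI, classed as indoor ambient sound.
# Each mix layers a residential interior bed at full level with the resort's night exterior loop turned well down.
# A quality-level node picks the quad mix for its first two levels and the stereo mix for the last two.
# Playback is non-spatialized and unattenuated, without reverb send, and its concurrency override prevents new instances at the limit.
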SoundCue /AmbientAudioContent/SoundCues/POI/Ambience_BeachResort_Medium_Room_Int_Night_Loop_Cue
root=QualityLevel_0
nodes [
 {id=Mixer_1, children [WavePlayer_0,WavePlayer_5], properties {inputVolume [0.2,1]}}
 {id=Mixer_3, children [WavePlayer_4,WavePlayer_1], properties {inputVolume [1,0.2]}}
 {id=QualityLevel_0, children [Mixer_1,Mixer_1,Mixer_3,Mixer_3]}
 {id=WavePlayer_0, properties {bLooping=true, soundWaveAssetPtr=/AmbientAudioContent/Waves/POI/Ambience_BeachResort_Night_Exterior_Loop_Quad}}
 {
  id=WavePlayer_1
  properties {bLooping=true, soundWaveAssetPtr=/AmbientAudioContent/Waves/POI/Ambience_BeachResort_Night_Exterior_Loop_Stereo}
 }
 {id=WavePlayer_4, properties {bLooping=true, soundWaveAssetPtr=/AmbientAudioContent/Waves/POI/Ambience_Residential_Int_Stereo_3}}
 {id=WavePlayer_5, properties {bLooping=true, soundWaveAssetPtr=/AmbientAudioContent/Waves/POI/Ambience_Residential_Int_Quad_3}}
]
properties {
 bOverrideConcurrency=true
 maxDistance=2097152
 soundClassObject=&Fort_Ambient_Indoors_Athena
 volumeMultiplier=1.403
 attenuationOverrides {bAttenuate=false, bEnableReverbSend=false, bSpatialize=false}
 concurrencyOverrides {resolutionRule=PreventNew}
}
