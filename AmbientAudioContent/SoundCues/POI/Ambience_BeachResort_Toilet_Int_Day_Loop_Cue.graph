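# Looping, non-spatialized day interior ambience cue for the Beach Resort toilet point of interest, routed to the indoor ambient sound class.
# A quality-level switch sends its first two levels to a mix with a quad residential interior bed and the last two to a stereo version.
# Both mixes layer that bed under a louder Starlight bathroom loop reused at a fixed lowered pitch.
# Attenuation and reverb send are off, and concurrency prevents new instances.
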SoundCue /AmbientAudioContent/SoundCues/POI/Ambience_BeachResort_Toilet_Int_Day_Loop_Cue
root=QualityLevel_0
nodes [
 {id=Mixer_0, children [Modulator_1,WavePlayer_0], properties {inputVolume [0.5,1.5]}}
 {id=Mixer_2, children [Modulator_2,WavePlayer_2], properties {inputVolume [0.5,1.5]}}
 {id=Modulator_1, children [WavePlayer_1], properties {pitchMax=0.6, pitchMin=0.6}}
 {id=Modulator_2, children [WavePlayer_3], properties {pitchMax=0.6, pitchMin=0.6}}
 {id=QualityLevel_0, children [Mixer_0,Mixer_0,Mixer_2,Mixer_2]}
 {id=WavePlayer_0, properties {bLooping=true, soundWaveAssetPtr=/AmbientAudioContent/Waves/POI/Ambience_Residential_Int_Quad_5}}
 {
  id=WavePlayer_1
  properties {bLooping=true, soundWaveAssetPtr=/AmbientAudioContent/Waves/POI/Ambience_Starlight_Bathroom_Int_Day_Loop_Stereo}
 }
 {id=WavePlayer_2, properties {bLooping=true, soundWaveAssetPtr=/AmbientAudioContent/Waves/POI/Ambience_Residential_Int_Stereo_5}}
 {
  id=WavePlayer_3
  properties {bLooping=true, soundWaveAssetPtr=&Ambience_Starlight_Bathroom_Int_Day_Loop_Stereo}
 }
]
properties {
 bOverrideConcurrency=true
 maxDistance=2097152
 soundClassObject=&Fort_Ambient_Indoors_Athena
 volumeMultiplier=1.3
 attenuationOverrides {bAttenuate=false, bEnableReverbSend=false, bSpatialize=false}
 concurrencyOverrides {resolutionRule=PreventNew}
}
