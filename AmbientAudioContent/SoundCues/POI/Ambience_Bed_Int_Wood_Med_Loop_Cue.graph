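# Looping indoor ambience bed cue for medium wood interiors at POIs, routed to the Athena indoor ambient sound class.
# A quality-level node picks the quad looping wave for the first two levels and the stereo one for the other two.
# Two nested, indefinitely looping envelopers add movement: the inner varies pitch over 30 seconds, the outer gently swells volume over 20.
# The differing cycle lengths keep the modulation from repeating in step.
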SoundCue /AmbientAudioContent/SoundCues/POI/Ambience_Bed_Int_Wood_Med_Loop_Cue
root=Enveloper_2
nodes [
 {
  id=Enveloper_1
  children [QualityLevel_0]
  properties {
   bLoop=true
   bLoopIndefinitely=true
   loopEnd=30
   pitchCurve {editorCurveData {keys [{interpMode=RCIM_Cubic, time=0}, <at 15.9468 0.9181 RCIM_Cubic>, {interpMode=RCIM_Cubic, time=30}]}}
   volumeCurve {editorCurveData {keys [<at 0 1>]}}
  }
 }
 {
  id=Enveloper_2
  children [Enveloper_1]
  properties {
   bLoop=true
   bLoopIndefinitely=true
   loopEnd=20
   pitchCurve {editorCurveData {keys [{time=0}]}}
   volumeCurve {editorCurveData {keys [
      {arriveTangent=0, interpMode=RCIM_Cubic, time=0, value=1}
      {arriveTangent=0, interpMode=RCIM_Cubic, leaveTangent=0, time=10, value=1.1}
      {interpMode=RCIM_Cubic, leaveTangent=0, time=20, value=0.9992}
     ]}}
  }
 }
 {id=QualityLevel_0, children [WavePlayer_0,WavePlayer_0,WavePlayer_1,WavePlayer_1]}
 {id=WavePlayer_0, properties {bLooping=true, soundWaveAssetPtr=/AmbientAudioContent/Waves/POI/Ambience_Bed_Int_Wood_Med_01_Quad}}
 {id=WavePlayer_1, properties {bLooping=true, soundWaveAssetPtr=/AmbientAudioContent/Waves/POI/Ambience_Bed_Int_Wood_Med_01_St}}
]
properties {
 attenuationSettings=&Fort_Ambient_Athena_Att
 maxDistance=2097152
 soundClassObject=&Fort_Ambient_Indoors_Athena
 volumeMultiplier=0.708
 attenuationOverrides {bAttenuate=false}
}
