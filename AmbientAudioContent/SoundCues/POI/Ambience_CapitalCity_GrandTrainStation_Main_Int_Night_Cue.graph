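# Looping night interior ambience cue for the Capital City Grand Train Station, played non-spatialized and unattenuated in the indoor ambient sound class.
# A quality-level node selects the looping quad wave for its first two slots and the stereo wave for the last two.
# Two nested, endlessly looping envelopers add slow variation: the inner curves pitch over 30 seconds, the outer gently dips volume over 20.
# Its concurrency override prevents new instances once the limit is reached.
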SoundCue /AmbientAudioContent/SoundCues/POI/Ambience_CapitalCity_GrandTrainStation_Main_Int_Night_Cue
root=Enveloper_3
nodes [
 {
  id=Enveloper_0
  children [QualityLevel_0]
  properties {
   bLoop=true
   bLoopIndefinitely=true
   loopEnd=30
   pitchCurve {editorCurveData {keys [{interpMode=RCIM_Cubic, time=0}, <at 15.9468 0.9181 RCIM_Cubic>, {interpMode=RCIM_Cubic, time=30}]}}
   volumeCurve {editorCurveData {keys [<at 0 1>]}}
  }
 }
 {
  id=Enveloper_3
  children [Enveloper_0]
  properties {
   bLoop=true
   bLoopIndefinitely=true
   loopEnd=20
   pitchCurve {editorCurveData {keys [{time=0}]}}
   volumeCurve {editorCurveData {keys [
      {arriveTangent=-0.01, interpMode=RCIM_Cubic, leaveTangent=-0.01, time=0, value=1}
      <at 10 0.9 RCIM_Cubic>
      {arriveTangent=0.01, interpMode=RCIM_Cubic, leaveTangent=0.01, time=20, value=1}
     ]}}
  }
 }
 {id=QualityLevel_0, children [WavePlayer_7,WavePlayer_7,WavePlayer_8,WavePlayer_8]}
 {
  id=WavePlayer_7
  properties {bLooping=true, soundWaveAssetPtr=/AmbientAudioContent/Waves/POI/Ambience_CapitalCity_GrandTrainStation_Main_Night_Int_Quad}
 }
 {
  id=WavePlayer_8
  properties {bLooping=true, soundWaveAssetPtr=/AmbientAudioContent/Waves/POI/Ambience_CapitalCity_GrandTrainStation_Main_Night_Int_Stereo}
 }
]
properties {
 bOverrideConcurrency=true
 maxDistance=2097152
 soundClassObject=&Fort_Ambient_Indoors_Athena
 volumeMultiplier=0.2
 attenuationOverrides {bAttenuate=false, bEnableReverbSend=false, bSpatialize=false}
 concurrencyOverrides {resolutionRule=PreventNew}
}
