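# Looping sound cue for the night interior ambience of a side room in the Capital City Grand Train Station, played on the indoor ambient sound class.
# A quality-level node sends the first two levels to a quad branch and the last two to a stereo branch.
# Each mixes pitched-down loops of quiet daytime main-station audio and boosted residential interior audio.
# Nested looping envelopes with different periods slowly drift pitch and volume to vary the loops.
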SoundCue /AmbientAudioContent/SoundCues/POI/Ambience_CapitalCity_GrandTrainStation_Side_Room_Night_Int_Cue
root=QualityLevel_0
nodes [
 {
  id=Enveloper_1
  children [Mixer_0]
  properties {
   bLoop=true
   bLoopIndefinitely=true
   loopEnd=30
   pitchCurve {editorCurveData {keys [{interpMode=RCIM_Cubic, time=0}, <at 15.9468 0.9181 RCIM_Cubic>, {interpMode=RCIM_Cubic, time=30}]}}
   volumeCurve {editorCurveData {keys [<at 0 1>]}}
  }
 }
 {
  id=Enveloper_2
  children [Enveloper_1]
  properties {
   bLoop=true
   bLoopIndefinitely=true
   loopEnd=20
   pitchCurve {editorCurveData {keys [{time=0}]}}
   volumeCurve {editorCurveData {keys [
      {arriveTangent=0, interpMode=RCIM_Cubic, leaveTangent=0, time=0, value=1}
      {arriveTangent=0, interpMode=RCIM_Cubic, leaveTangent=0, time=10, value=0.9}
      {arriveTangent=0, interpMode=RCIM_Cubic, leaveTangent=0, time=20, value=0.9992}
     ]}}
  }
 }
 {
  id=Enveloper_4
  children [Mixer_1]
  properties {
   bLoop=true
   bLoopIndefinitely=true
   loopEnd=30
   pitchCurve {editorCurveData {keys [{interpMode=RCIM_Cubic, time=0}, <at 15.9468 0.9181 RCIM_Cubic>, {interpMode=RCIM_Cubic, time=30}]}}
   volumeCurve {editorCurveData {keys [<at 0 1>]}}
  }
 }
 {
  id=Enveloper_5
  children [Enveloper_4]
  properties {
   bLoop=true
   bLoopIndefinitely=true
   loopEnd=20
   pitchCurve {editorCurveData {keys [{time=0}]}}
   volumeCurve {editorCurveData {keys [
      {arriveTangent=0, interpMode=RCIM_Cubic, leaveTangent=0, time=0, value=1}
      {arriveTangent=0, interpMode=RCIM_Cubic, leaveTangent=0, time=10, value=0.9}
      {arriveTangent=0, interpMode=RCIM_Cubic, leaveTangent=0, time=20, value=0.9992}
     ]}}
  }
 }
 {id=Mixer_0, children [Modulator_1,Modulator_0], properties {inputVolume [1,1]}}
 {id=Mixer_1, children [Modulator_3,Modulator_2], properties {inputVolume [1,1]}}
 {id=Modulator_0, children [WavePlayer_7], properties {pitchMax=0.9, pitchMin=0.9, volumeMax=3, volumeMin=3}}
 {id=Modulator_1, children [WavePlayer_4], properties {pitchMax=0.9, pitchMin=0.9, volumeMax=0.4, volumeMin=0.4}}
 {id=Modulator_2, children [WavePlayer_6], properties {pitchMax=0.9, pitchMin=0.9, volumeMax=3, volumeMin=3}}
 {id=Modulator_3, children [WavePlayer_5], properties {pitchMax=0.9, pitchMin=0.9, volumeMax=0.4, volumeMin=0.4}}
 {id=QualityLevel_0, children [Enveloper_2,Enveloper_2,Enveloper_5,Enveloper_5]}
 {
  id=WavePlayer_4
  properties {bLooping=true, soundWaveAssetPtr=/AmbientAudioContent/Waves/POI/Ambience_CapitalCity_GrandTrainStation_Main_Int_Day_Quad}
 }
 {
  id=WavePlayer_5
  properties {bLooping=true, soundWaveAssetPtr=/AmbientAudioContent/Waves/POI/Ambience_CapitalCity_GrandTrainStation_Main_Int_Day_Stereo}
 }
 {id=WavePlayer_6, properties {bLooping=true, soundWaveAssetPtr=/AmbientAudioContent/Waves/POI/Ambience_Residential_Int_Stereo_3}}
 {id=WavePlayer_7, properties {bLooping=true, soundWaveAssetPtr=/AmbientAudioContent/Waves/POI/Ambience_Residential_Int_Quad_3}}
]
properties {
 attenuationSettings=&Fort_Ambient_Athena_Att
 maxDistance=2097152
 soundClassObject=&Fort_Ambient_Indoors_Athena
 volumeMultiplier=0.251
 preEffectBusSends [{}]
}
