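# Looping exterior morning city ambience cue for the CrimeTown POI, routed to the outdoor ambient sound class and built from Day wave assets.
# A quality-level node sends its first two slots to a random pick between two looping quad waves and its last two to one looping stereo wave.
# Both branches pass through two nested, endlessly looping envelopers that lower the pitch, add cyclic pitch movement and gently dip the volume.
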SoundCue /AmbientAudioContent/SoundCues/POI/Ambience_CrimeTown_City_Morning_Loop_Ext_Cue
root=QualityLevel_0
nodes [
 {
  id=Enveloper_0
  children [Enveloper_3]
  properties {
   bLoop=true
   bLoopIndefinitely=true
   loopEnd=100
   pitchCurve {editorCurveData {keys [<at 0 0.8>]}}
   volumeCurve {editorCurveData {keys [<at 0 1 RCIM_Cubic>,<at 50 0.9 RCIM_Cubic>,<at 100 1 RCIM_Cubic>]}}
  }
 }
 {
  id=Enveloper_3
  children [Random_1]
  properties {
   bLoop=true
   bLoopIndefinitely=true
   loopEnd=30
   pitchCurve {editorCurveData {keys [{interpMode=RCIM_Cubic, time=0}, <at 15.9468 0.9181 RCIM_Cubic>, {interpMode=RCIM_Cubic, time=30}]}}
   volumeCurve {editorCurveData {keys [<at 0 1>]}}
  }
 }
 {
  id=Enveloper_5
  children [Enveloper_6]
  properties {
   bLoop=true
   bLoopIndefinitely=true
   loopEnd=100
   pitchCurve {editorCurveData {keys [<at 0 0.8>]}}
   volumeCurve {editorCurveData {keys [<at 0 1 RCIM_Cubic>,<at 50 0.9 RCIM_Cubic>,<at 100 1 RCIM_Cubic>]}}
  }
 }
 {
  id=Enveloper_6
  children [WavePlayer_4]
  properties {
   bLoop=true
   bLoopIndefinitely=true
   loopEnd=30
   pitchCurve {editorCurveData {keys [{interpMode=RCIM_Cubic, time=0}, <at 15.9468 0.9181 RCIM_Cubic>, {interpMode=RCIM_Cubic, time=30}]}}
   volumeCurve {editorCurveData {keys [<at 0 1>]}}
  }
 }
 {id=QualityLevel_0, children [Enveloper_0,Enveloper_0,Enveloper_5,Enveloper_5]}
 {id=Random_1, children [WavePlayer_5,WavePlayer_1], properties {weights [1,1]}}
 {id=WavePlayer_1, properties {bLooping=true, soundWaveAssetPtr=/AmbientAudioContent/Waves/POI/Ambience_CrimeTown_City_Day_02_Quad_Loop}}
 {id=WavePlayer_4, properties {bLooping=true, soundWaveAssetPtr=/AmbientAudioContent/Waves/POI/Ambience_CrimeTown_City_Day_01_Stereo_Loop}}
 {id=WavePlayer_5, properties {bLooping=true, soundWaveAssetPtr=/AmbientAudioContent/Waves/POI/Ambience_CrimeTown_City_Day_01_Quad_Loop}}
]
properties {maxDistance=2097152, soundClassObject=&Fort_Ambient_Outdoors_Athena, volumeMultiplier=0.45}
