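# Looping outdoor ambience cue for the morning/day magic layer of the Frog Castle point of interest.
# Quality levels select quad waves for the first two slots and stereo for the last two.
# Each branch mixes a mysterious wind wave with a pitched-down landmark loop, with the relative weights swapped between quad and stereo.
# Nested endlessly looping envelopers modulate pitch and volume on different periods, so the layer drifts over time.
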SoundCue /AmbientAudioContent/SoundCues/POI/Ambience_Frog_Caslte_Magic_Layer_MorningDay_Cue
root=QualityLevel_0
nodes [
 {
  id=Enveloper_0
  children [Mixer_0]
  properties {
   bLoop=true
   bLoopIndefinitely=true
   loopEnd=30
   pitchCurve {editorCurveData {keys [
      {arriveTangent=-0.0013, interpMode=RCIM_Cubic, leaveTangent=-0.0013, time=0}
      <at 15 0.98 RCIM_Cubic>
      {arriveTangent=0.0013, interpMode=RCIM_Cubic, leaveTangent=0.0013, time=30}
     ]}}
   volumeCurve {editorCurveData {keys [<at 0 1>]}}
  }
 }
 {
  id=Enveloper_1
  children [Mixer_1]
  properties {
   bLoop=true
   bLoopIndefinitely=true
   loopEnd=30
   pitchCurve {editorCurveData {keys [
      {arriveTangent=-0.0013, interpMode=RCIM_Cubic, leaveTangent=-0.0013, time=0}
      <at 15 0.98 RCIM_Cubic>
      {arriveTangent=0.0013, interpMode=RCIM_Cubic, leaveTangent=0.0013, time=30}
     ]}}
   volumeCurve {editorCurveData {keys [<at 0 1>]}}
  }
 }
 {
  id=Enveloper_2
  children [Enveloper_1]
  properties {
   bLoop=true
   bLoopIndefinitely=true
   loopEnd=20
   pitchCurve {editorCurveData {keys [{time=0}]}}
   volumeCurve {editorCurveData {keys [
      {arriveTangent=-0.01, interpMode=RCIM_Cubic, leaveTangent=-0.01, time=0, value=1}
      <at 10 0.9 RCIM_Cubic>
      {arriveTangent=0.01, interpMode=RCIM_Cubic, leaveTangent=0.01, time=20, value=1}
     ]}}
  }
 }
 {
  id=Enveloper_3
  children [Enveloper_0]
  properties {
   bLoop=true
   bLoopIndefinitely=true
   loopEnd=20
   pitchCurve {editorCurveData {keys [{time=0}]}}
   volumeCurve {editorCurveData {keys [
      {arriveTangent=-0.01, interpMode=RCIM_Cubic, leaveTangent=-0.01, time=0, value=1}
      <at 10 0.9 RCIM_Cubic>
      {arriveTangent=0.01, interpMode=RCIM_Cubic, leaveTangent=0.01, time=20, value=1}
     ]}}
  }
 }
 {id=Mixer_0, children [WavePlayer_2,Modulator_0], properties {inputVolume [1,0.5]}}
 {id=Mixer_1, children [Modulator_1,WavePlayer_3], properties {inputVolume [1,0.5]}}
 {id=Modulator_0, children [WavePlayer_0], properties {pitchMax=0.7, pitchMin=0.7}}
 {id=Modulator_1, children [WavePlayer_1], properties {pitchMax=0.7, pitchMin=0.7}}
 {id=QualityLevel_0, children [Enveloper_3,Enveloper_3,Enveloper_2,Enveloper_2]}
 {id=WavePlayer_0, properties {bLooping=true, soundWaveAssetPtr=/AmbientAudioContent/Waves/POI/Ambience_Japan_Artist_CollabLandmark_Loop_Quad}}
 {
  id=WavePlayer_1
  properties {bLooping=true, soundWaveAssetPtr=/AmbientAudioContent/Waves/POI/Ambience_Japan_Artist_CollabLandmark_Loop_Stereo}
 }
 {id=WavePlayer_2, properties {soundWaveAssetPtr=/AmbientAudioContent/Waves/POI/Ambience_Henchmen_DigSite_Mysterious_Wind_Day_Loop_Quad}}
 {id=WavePlayer_3, properties {soundWaveAssetPtr=/AmbientAudioContent/Waves/POI/Ambience_Henchmen_DigSite_Mysterious_Wind_Day_Loop_Stereo}}
]
properties {
 attenuationSettings=&Fort_Ambient_Athena_Att
 maxDistance=2097152
 soundClassObject=&Fort_Ambient_Outdoors_Athena
 volumeMultiplier=0.501
 preEffectBusSends [{}]
}
