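# Looping, non-spatialized ambience bed for a gold mine metal cave POI, played quietly and pitched down in the ambient sound class.
# A quality-level node picks a quad mix for the first two levels and a stereo mix for the other two.
# Each mix layers mine, pitched-down industrial and airborne-city underside loops with randomized modulation.
# Two nested looping envelopers of different lengths add slow volume and pitch drift over the selected mix.
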SoundCue /AmbientAudioContent/SoundCues/POI/Ambience_GoldMine_Metal_Cave_Loop_Cue
root=Enveloper_3
nodes [
 {
  id=Enveloper_0
  children [QualityLevel_1]
  properties {
   bLoop=true
   bLoopIndefinitely=true
   loopEnd=30
   pitchCurve {editorCurveData {keys [{interpMode=RCIM_Cubic, time=0}, <at 15.9468 0.9181 RCIM_Cubic>, {interpMode=RCIM_Cubic, time=30}]}}
   volumeCurve {editorCurveData {keys [<at 0 1>]}}
  }
 }
 {
  id=Enveloper_3
  children [Enveloper_0]
  properties {
   bLoop=true
   bLoopIndefinitely=true
   loopEnd=20
   pitchCurve {editorCurveData {keys [{time=0}]}}
   volumeCurve {editorCurveData {keys [
      {arriveTangent=-0.01, interpMode=RCIM_Cubic, leaveTangent=-0.01, time=0, value=1}
      <at 10 0.9 RCIM_Cubic>
      {arriveTangent=0.01, interpMode=RCIM_Cubic, leaveTangent=0.01, time=20, value=1}
     ]}}
  }
 }
 {id=Mixer_0, children [Modulator_2,WavePlayer_4,WavePlayer_1], properties {inputVolume [0.5,1,1]}}
 {id=Mixer_1, children [WavePlayer_5,Modulator_0,Modulator_12], properties {inputVolume [1,1,1]}}
 {id=Modulator_0, children [WavePlayer_0], properties {pitchMax=0.8, pitchMin=0.5}}
 {id=Modulator_1, children [Mixer_0], properties {pitchMax=1.2, pitchMin=0.8, volumeMax=1.1, volumeMin=0.9}}
 {id=Modulator_12, children [WavePlayer_2], properties {pitchMax=1.2, pitchMin=0.8, volumeMax=1.1, volumeMin=0.9}}
 {id=Modulator_2, children [WavePlayer_3], properties {pitchMax=0.8, pitchMin=0.5}}
 {id=QualityLevel_1, children [Mixer_1,Mixer_1,Modulator_1,Modulator_1]}
 {id=WavePlayer_0, properties {bLooping=true, soundWaveAssetPtr=/AmbientAudioContent/Waves/POI/Ambience_IndustrialLarge001_Loop_Quad}}
 {
  id=WavePlayer_1
  properties {bLooping=true, soundWaveAssetPtr=/Game/Sounds/Artemis/POI/AirborneCity/Ambience_AirborneCity_Underisde_Night_Stereo}
 }
 {id=WavePlayer_2, properties {bLooping=true, soundWaveAssetPtr=/AmbientAudioContent/Waves/POI/Ambience_MineA012_Loop_Quad}}
 {id=WavePlayer_3, properties {bLooping=true, soundWaveAssetPtr=/AmbientAudioContent/Waves/POI/Ambience_IndustrialLarge001_Loop_Stereo}}
 {id=WavePlayer_4, properties {bLooping=true, soundWaveAssetPtr=/AmbientAudioContent/Waves/POI/Ambience_MineA012_Loop_Stereo}}
 {
  id=WavePlayer_5
  properties {bLooping=true, soundWaveAssetPtr=/Game/Sounds/Artemis/POI/AirborneCity/Ambience_AirborneCity_Underisde_Night_Quad}
 }
]
properties {
 bOverrideConcurrency=true
 maxDistance=2097152
 pitchMultiplier=0.8
 soundClassObject=&Fort_Ambient_Athena
 volumeMultiplier=0.13
 attenuationOverrides {bAttenuate=false, bEnableReverbSend=false, bSpatialize=false}
 concurrencyOverrides {resolutionRule=PreventNew}
}
