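# Looping exterior wind ambience cue for the haunted house POI, assigned to the outdoor ambient sound class with shared ambient attenuation.
# A single looping exterior wave is boosted by a mixer, then passes through two nested, endlessly looping Envelopers.
# The inner Enveloper modulates pitch over a 60-second cycle and the outer one gently swells volume over 20 seconds, adding slow movement to the one wave.
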SoundCue /AmbientAudioContent/SoundCues/POI/Ambience_Haunted_House_Wind_Ext_Cue
root=Enveloper_3
nodes [
 {
  id=Enveloper_0
  children [Mixer_1]
  properties {
   bLoop=true
   bLoopIndefinitely=true
   loopEnd=60
   pitchCurve {editorCurveData {keys [
      {arriveTangent=-0.0033, interpMode=RCIM_Cubic, time=0}
      <at 15 0.95 RCIM_Cubic>
      {arriveTangent=0.005, interpMode=RCIM_Cubic, leaveTangent=0.005, time=30}
      <at 45 1.1 RCIM_Cubic>
      {time=60}
     ]}}
   volumeCurve {editorCurveData {keys [<at 0 1>]}}
  }
 }
 {
  id=Enveloper_3
  children [Enveloper_0]
  properties {
   bLoop=true
   bLoopIndefinitely=true
   loopEnd=20
   pitchCurve {editorCurveData {keys [{time=0}]}}
   volumeCurve {editorCurveData {keys [
      {arriveTangent=-0.01, interpMode=RCIM_Cubic, time=0, value=1}
      <at 10 1.1 RCIM_Cubic>
      {interpMode=RCIM_Cubic, leaveTangent=0.01, time=20, value=1}
     ]}}
  }
 }
 {id=Mixer_1, children [WavePlayer_3], properties {inputVolume [1.8]}}
 {id=WavePlayer_3, properties {bLooping=true, soundWaveAssetPtr=/AmbientAudioContent/Waves/POI/Ambience_HauntedHouse_Exterior_Main_02}}
]
properties {
 attenuationSettings=/AmbientAudioBase/Attenuations/Fort_Ambient_Att
 maxDistance=2097152
 soundClassObject=&Fort_Ambient_Outdoors_Athena
 volumeMultiplier=0.25
 preEffectBusSends [{}]
}
