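# Sound cue for the looping daytime interior ambience of the Hollow Base bathroom point of interest.
# A single looping wave player feeds the output at reduced volume through the Athena indoor ambient sound class.
# Attenuation overrides turn off distance falloff, spatialization and reverb send, so it plays as a non-positional 2D bed.
# A concurrency override prevents new instances from starting once the limit is reached.
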SoundCue /AmbientAudioContent/SoundCues/POI/Ambience_Hollow_Base_Bathroom_Int_Day_Loop_Cue
root=WavePlayer_1
nodes [{id=WavePlayer_1, properties {bLooping=true, soundWaveAssetPtr=/AmbientAudioContent/Waves/POI/Ambience_Hollow_Base_Bathroom_Int_Day_Loop}}]
properties {
 bOverrideConcurrency=true
 maxDistance=2097152
 soundClassObject=&Fort_Ambient_Indoors_Athena
 volumeMultiplier=0.626
 attenuationOverrides {bAttenuate=false, bEnableReverbSend=false, bSpatialize=false}
 concurrencyOverrides {resolutionRule=PreventNew}
}
