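# Looping, non-spatialized indoor ambience cue, named for the HydroBottle lab corridor POI, with attenuation and reverb send off.
# A quality-level node picks between two mixes, pairing the Underground Lair cave loop with a Hollow Base corridor or bathroom loop.
# Two nested looping envelopers process the result: the inner varies pitch over a shorter cycle, while the outer lowers pitch and slowly swells volume.
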SoundCue /AmbientAudioContent/SoundCues/POI/Ambience_HydroBottle_Lab_Corrdor_Loop_Cue
root=Enveloper_6
nodes [
 {
  id=Enveloper_0
  children [QualityLevel_0]
  properties {
   bLoop=true
   bLoopIndefinitely=true
   loopEnd=30
   pitchCurve {editorCurveData {keys [{interpMode=RCIM_Cubic, time=0}, <at 15.9468 0.9181 RCIM_Cubic>, {interpMode=RCIM_Cubic, time=30}]}}
   volumeCurve {editorCurveData {keys [<at 0 1>]}}
  }
 }
 {
  id=Enveloper_6
  children [Enveloper_0]
  properties {
   bLoop=true
   bLoopIndefinitely=true
   loopEnd=100
   pitchCurve {editorCurveData {keys [<at 0 0.8>]}}
   volumeCurve {editorCurveData {keys [<at 0 1 RCIM_Cubic>,<at 50 1.2 RCIM_Cubic>,<at 100 1 RCIM_Cubic>]}}
  }
 }
 {id=Mixer_1, children [WavePlayer_2,WavePlayer_5], properties {inputVolume [1.7,1]}}
 {id=Mixer_3, children [WavePlayer_4,WavePlayer_3], properties {inputVolume [1,1.7]}}
 {id=QualityLevel_0, children [Mixer_1,Mixer_1,Mixer_3,Mixer_3]}
 {
  id=WavePlayer_2
  properties {bLooping=true, soundWaveAssetPtr=/Game/Sounds/Helios/POI/UndergroundLair/Ambience_Underground_Lair_Cave_Main_Airy_Quad_Loop}
 }
 {
  id=WavePlayer_3
  properties {bLooping=true, soundWaveAssetPtr=&Ambience_Underground_Lair_Cave_Main_Airy_Quad_Loop}
 }
 {
  id=WavePlayer_4
  properties {bLooping=true, soundWaveAssetPtr=/Game/Sounds/Artemis/POI/Hollow_Base/Ambience_Hollow_Base_Bathroom_Int_Day_Loop}
 }
 {id=WavePlayer_5, properties {bLooping=true, soundWaveAssetPtr=/Game/Sounds/Artemis/POI/Hollow_Base/Ambience_Hollow_Base_Coridoor_Day_Quad}}
]
properties {
 bOverrideConcurrency=true
 maxDistance=2097152
 soundClassObject=&Fort_Ambient_Indoors_Athena
 volumeMultiplier=0.7
 attenuationOverrides {bAttenuate=false, bEnableReverbSend=false, bSpatialize=false}
 concurrencyOverrides {resolutionRule=PreventNew}
}
