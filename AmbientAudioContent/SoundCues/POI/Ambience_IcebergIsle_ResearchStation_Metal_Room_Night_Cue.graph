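# Looping, non-spatialized indoor ambience cue for the Iceberg Isle research station's metal room at night, with attenuation disabled.
# A shared industrial and crypt-interior bed is mixed with day and night dock-interior loops reused from other locations.
# The quality switch selects quad dock loops for the first two levels and stereo for the last two.
# Nested looping envelopers add slow drift: a 30-second pitch curve inside a gentle 20-second volume swell.
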SoundCue /AmbientAudioContent/SoundCues/POI/Ambience_IcebergIsle_ResearchStation_Metal_Room_Night_Cue
root=Enveloper_3
nodes [
 {
  id=Enveloper_0
  children [QualityLevel_1]
  properties {
   bLoop=true
   bLoopIndefinitely=true
   loopEnd=30
   pitchCurve {editorCurveData {keys [{interpMode=RCIM_Cubic, time=0}, <at 15.9468 0.9181 RCIM_Cubic>, {interpMode=RCIM_Cubic, time=30}]}}
   volumeCurve {editorCurveData {keys [<at 0 1>]}}
  }
 }
 {
  id=Enveloper_3
  children [Enveloper_0]
  properties {
   bLoop=true
   bLoopIndefinitely=true
   loopEnd=20
   pitchCurve {editorCurveData {keys [{time=0}]}}
   volumeCurve {editorCurveData {keys [
      {arriveTangent=-0.01, interpMode=RCIM_Cubic, leaveTangent=-0.01, time=0, value=1}
      <at 10 0.9 RCIM_Cubic>
      {arriveTangent=0.01, interpMode=RCIM_Cubic, leaveTangent=0.01, time=20, value=1}
     ]}}
  }
 }
 {id=Mixer_0, children [WavePlayer_3,WavePlayer_2], properties {inputVolume [0.5,2]}}
 {id=Mixer_2, children [WavePlayer_5,Mixer_0,WavePlayer_8], properties {inputVolume [0.2,0.75,0.3]}}
 {id=Mixer_3, children [WavePlayer_4,Mixer_0,WavePlayer_9], properties {inputVolume [0.2,0.75,0.3]}}
 {id=QualityLevel_1, children [Mixer_2,Mixer_2,Mixer_3,Mixer_3]}
 {
  id=WavePlayer_2
  properties {bLooping=true, soundWaveAssetPtr=/AmbientAudioContent/Waves/POI/Ambience_Starlight_Crypt_Int_Morning_02_Loop_Stereo}
 }
 {id=WavePlayer_3, properties {bLooping=true, soundWaveAssetPtr=/AmbientAudioContent/Waves/POI/Ambience_IndustrialSmall002_Loop_Stereo}}
 {
  id=WavePlayer_4
  properties {bLooping=true, soundWaveAssetPtr=/AmbientAudioContent/Waves/POI/Ambience_Hollow_Base_Takeover_Dock_Interior_Day_Loop_Stereo}
 }
 {
  id=WavePlayer_5
  properties {bLooping=true, soundWaveAssetPtr=/AmbientAudioContent/Waves/POI/Ambience_Hollow_Base_Takeover_Dock_Interior_Day_Loop_Quad}
 }
 {
  id=WavePlayer_8
  properties {bLooping=true, soundWaveAssetPtr=/AmbientAudioContent/Waves/POI/Ambience_Hollow_Base_Takeover_Dock_Interior_Night_Loop_Quad}
 }
 {
  id=WavePlayer_9
  properties {bLooping=true, soundWaveAssetPtr=/AmbientAudioContent/Waves/POI/Ambience_Hollow_Base_Takeover_Dock_Interior_Night_Loop_Stereo}
 }
]
properties {
 bOverrideConcurrency=true
 maxDistance=2097152
 soundClassObject=&Fort_Ambient_Indoors_Athena
 volumeMultiplier=0.351
 attenuationOverrides {bAttenuate=false, bEnableReverbSend=false, bSpatialize=false}
 concurrencyOverrides {resolutionRule=PreventNew}
}
