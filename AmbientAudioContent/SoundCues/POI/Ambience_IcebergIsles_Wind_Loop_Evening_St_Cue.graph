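# Looping evening wind ambience cue for the Iceberg Isles POI, played non-spatialized in the outdoor ambient sound class.
# A mixer layers a winter night bed, a quieter spooky wind loop and a randomly modulated desert night wind.
# Looping envelopers with differing periods drift the layers, using bell and reversed-bell curves to trade the first two.
# All quality levels share that mix, beneath two nested envelopers that slowly vary the whole output.
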SoundCue /AmbientAudioContent/SoundCues/POI/Ambience_IcebergIsles_Wind_Loop_Evening_St_Cue
root=Enveloper_6
nodes [
 {
  id=Enveloper_0
  children [WavePlayer_3]
  properties {
   bLoop=true
   bLoopIndefinitely=true
   loopEnd=30
   pitchCurve {editorCurveData {keys [{interpMode=RCIM_Cubic, time=0}, <at 15.9468 0.9181 RCIM_Cubic>, {interpMode=RCIM_Cubic, time=30}]}}
   volumeCurve {editorCurveData {keys [<at 0 1 RCIM_Cubic>,<at 16.435 0.8171 RCIM_Cubic>,<at 30 1 RCIM_Cubic>]}}
  }
 }
 {
  id=Enveloper_1
  children [QualityLevel_0]
  properties {
   bLoop=true
   bLoopIndefinitely=true
   loopEnd=40
   pitchCurve {editorCurveData {keys [
      {arriveTangent=0, interpMode=RCIM_Cubic, leaveTangent=0, tangentMode=RCTM_User, time=0}
      {arriveTangent=0, interpMode=RCIM_Cubic, leaveTangent=0, tangentMode=RCTM_User, time=20, value=0.92}
      {arriveTangent=0.004, interpMode=RCIM_Cubic, leaveTangent=0.004, time=40}
     ]}}
   volumeCurve {editorCurveData {keys [<at 0 1>]}}
  }
 }
 {
  id=Enveloper_6
  children [Enveloper_1]
  properties {
   bLoop=true
   bLoopIndefinitely=true
   loopEnd=30
   pitchCurve {editorCurveData {keys [{time=0}]}}
   volumeCurve {editorCurveData {keys [
      <at 0 1 RCIM_Cubic>
      {arriveTangent=0, interpMode=RCIM_Cubic, leaveTangent=0, tangentMode=RCTM_User, time=15, value=1.1}
      <at 30 1 RCIM_Cubic>
     ]}}
  }
 }
 {
  id=Enveloper_8
  children [WavePlayer_0]
  properties {
   bLoop=true
   bLoopIndefinitely=true
   loopEnd=100
   pitchCurve {editorCurveData {keys [<at 0 0.8>]}}
   volumeCurve {
    externalCurve=/Game/Sounds/Trans_Enveloper_Curves/BellReversed_100s.BellReversed_100S
    editorCurveData {keys [<at 0 1 RCIM_Cubic>,<at 50 0.5 RCIM_Cubic>,<at 100 1 RCIM_Cubic>]}
   }
  }
 }
 {
  id=Enveloper_9
  children [WavePlayer_1]
  properties {
   bLoop=true
   bLoopIndefinitely=true
   loopEnd=100
   pitchCurve {editorCurveData {keys [{time=0}]}}
   volumeCurve {
    externalCurve=/Game/Sounds/Trans_Enveloper_Curves/Bell_100s
    editorCurveData {keys [<at 0 0.1 RCIM_Cubic>,<at 100 1 RCIM_Cubic>,<at 200 0.1 RCIM_Cubic>]}
   }
  }
 }
 {id=Mixer_1, children [Enveloper_8,Enveloper_9,Modulator_0], properties {inputVolume [1,0.3,1]}}
 {id=Modulator_0, children [Enveloper_0], properties {pitchMax=1.1, pitchMin=0.9, volumeMax=1.1, volumeMin=0.9}}
 {id=QualityLevel_0, children [Mixer_1,Mixer_1,Mixer_1,Mixer_1]}
 {id=WavePlayer_0, properties {bLooping=true, soundWaveAssetPtr=/AmbientAudioContent/Waves/Biomes/Ambience_Winter_Night_Loop}}
 {id=WavePlayer_1, properties {bLooping=true, soundWaveAssetPtr=/AmbientAudioContent/Waves/SpottedLoops/Spot_WindSpooky01_Loop_Stereo}}
 {id=WavePlayer_3, properties {bLooping=true, soundWaveAssetPtr=/AmbientAudioContent/Waves/Biomes/Ambience_desert_wind_Night_02}}
]
properties {
 bOverrideConcurrency=true
 maxDistance=2097152
 soundClassObject=&Fort_Ambient_Outdoors_Athena
 volumeMultiplier=0.6
 attenuationOverrides {bSpatialize=false, falloffDistance=3000, attenuationShapeExtents {x=6000}}
 concurrencyOverrides {resolutionRule=PreventNew}
}
