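# Looping nighttime interior ambience cue for a medium lair POI, played on the Athena indoor ambient sound class.
# A mixer layers haunted-house interior and creak loops, a quieter lair night exterior bed, and a second, pitched-down copy of the creaks.
# A quality-level node picks the quad or stereo exterior wave.
# The interior and exterior layers pass through nested looping envelopes of different lengths, varying pitch and volume to mask repetition.
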SoundCue /AmbientAudioContent/SoundCues/POI/Ambience_Lair_Med_Int_Night_Cue
root=Mixer_1
nodes [
 {
  id=Enveloper_12
  children [WavePlayer_8]
  properties {
   bLoop=true
   bLoopIndefinitely=true
   loopEnd=30
   pitchCurve {editorCurveData {keys [{interpMode=RCIM_Cubic, time=0}, <at 15.9468 0.9181 RCIM_Cubic>, {interpMode=RCIM_Cubic, time=30}]}}
   volumeCurve {editorCurveData {keys [<at 0 1>]}}
  }
 }
 {
  id=Enveloper_13
  children [Enveloper_12]
  properties {
   bLoop=true
   bLoopIndefinitely=true
   loopEnd=20
   pitchCurve {editorCurveData {keys [{time=0}]}}
   volumeCurve {editorCurveData {keys [
      {arriveTangent=-0.01, interpMode=RCIM_Cubic, leaveTangent=-0.01, time=0, value=1}
      <at 10 0.9 RCIM_Cubic>
      {arriveTangent=0.01, interpMode=RCIM_Cubic, leaveTangent=0.01, time=20, value=1}
     ]}}
  }
 }
 {
  id=Enveloper_14
  children [Mixer_2]
  properties {
   bLoop=true
   bLoopIndefinitely=true
   loopEnd=30
   pitchCurve {editorCurveData {keys [{interpMode=RCIM_Cubic, time=0}, <at 15.9468 0.9181 RCIM_Cubic>, {interpMode=RCIM_Cubic, time=30}]}}
   volumeCurve {editorCurveData {keys [<at 0 1>]}}
  }
 }
 {
  id=Enveloper_15
  children [Enveloper_14]
  properties {
   bLoop=true
   bLoopIndefinitely=true
   loopEnd=20
   pitchCurve {editorCurveData {keys [{time=0}]}}
   volumeCurve {editorCurveData {keys [
      {arriveTangent=-0.01, interpMode=RCIM_Cubic, leaveTangent=-0.01, time=0, value=1}
      <at 10 0.9 RCIM_Cubic>
      {arriveTangent=0.01, interpMode=RCIM_Cubic, leaveTangent=0.01, time=20, value=1}
     ]}}
  }
 }
 {
  id=Enveloper_8
  children [WavePlayer_6]
  properties {
   bLoop=true
   bLoopIndefinitely=true
   loopEnd=30
   pitchCurve {editorCurveData {keys [{interpMode=RCIM_Cubic, time=0}, <at 15.9468 0.9181 RCIM_Cubic>, {interpMode=RCIM_Cubic, time=30}]}}
   volumeCurve {editorCurveData {keys [<at 0 1>]}}
  }
 }
 {
  id=Enveloper_9
  children [Enveloper_8]
  properties {
   bLoop=true
   bLoopIndefinitely=true
   loopEnd=20
   pitchCurve {editorCurveData {keys [{time=0}]}}
   volumeCurve {editorCurveData {keys [
      {arriveTangent=-0.01, interpMode=RCIM_Cubic, leaveTangent=-0.01, time=0, value=1}
      <at 10 0.9 RCIM_Cubic>
      {arriveTangent=0.01, interpMode=RCIM_Cubic, leaveTangent=0.01, time=20, value=1}
     ]}}
  }
 }
 {id=Mixer_1, children [Enveloper_15, QualityLevel_2, Modulator_1], properties {inputVolume [0.75,0.3,0.2]}}
 {id=Mixer_2, children [WavePlayer_9, WavePlayer_10], properties {inputVolume [0.7,0.6]}}
 {id=Modulator_1, children [WavePlayer_11], properties {pitchMax=0.8, pitchMin=0.8}}
 {id=QualityLevel_2, children [Enveloper_9,Enveloper_9,Enveloper_13,Enveloper_13]}
 {
  id=WavePlayer_10
  properties {bLooping=true, soundWaveAssetPtr=/AmbientAudioContent/Waves/POI/Ambience_HauntedHouse_AmbientLoop_Interior_Creaks_01}
 }
 {
  id=WavePlayer_11
  properties {bLooping=true, soundWaveAssetPtr=&Ambience_HauntedHouse_AmbientLoop_Interior_Creaks_01}
 }
 {id=WavePlayer_6, properties {bLooping=true, soundWaveAssetPtr=/AmbientAudioContent/Waves/POI/Ambience_Lair_Night_Ext_Loop_Quad}}
 {id=WavePlayer_8, properties {bLooping=true, soundWaveAssetPtr=/AmbientAudioContent/Waves/POI/Ambience_Lair_Night_Ext_Loop_Stereo}}
 {id=WavePlayer_9, properties {bLooping=true, soundWaveAssetPtr=/AmbientAudioContent/Waves/POI/Ambience_HauntedHouse_AmbientLoop_Interior_01}}
]
properties {
 maxDistance=2097152
 soundClassObject=&Fort_Ambient_Indoors_Athena
 volumeMultiplier=0.301
 preEffectBusSends [{}]
}
