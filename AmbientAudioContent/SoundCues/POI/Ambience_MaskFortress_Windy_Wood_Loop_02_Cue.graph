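# Looping indoor ambience cue for the MaskFortress POI, layering a windy wood interior with bangs over a wood attic room tone.
# A quality-level node sends its first two slots to quad waves and its last two to stereo versions, each pair mixed equally.
# Nested looping envelopes on each branch add a 40-second pitch curve and a gentle 30-second volume swell to vary the static loops.
# It uses indoor ambient class and attenuation, with concurrency overridden to one instance.
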SoundCue /AmbientAudioContent/SoundCues/POI/Ambience_MaskFortress_Windy_Wood_Loop_02_Cue
root=QualityLevel_1
nodes [
 {
  id=Enveloper_0
  children [Mixer_0]
  properties {
   bLoop=true
   bLoopIndefinitely=true
   loopEnd=40
   pitchCurve {editorCurveData {keys [
      {arriveTangent=0, interpMode=RCIM_Cubic, leaveTangent=0, tangentMode=RCTM_User, time=0}
      {arriveTangent=0, interpMode=RCIM_Cubic, leaveTangent=0, tangentMode=RCTM_User, time=20, value=0.92}
      {arriveTangent=0.004, interpMode=RCIM_Cubic, leaveTangent=0.004, time=40}
     ]}}
   volumeCurve {editorCurveData {keys [<at 0 1>]}}
  }
 }
 {
  id=Enveloper_4
  children [Mixer_3]
  properties {
   bLoop=true
   bLoopIndefinitely=true
   loopEnd=40
   pitchCurve {editorCurveData {keys [
      {arriveTangent=0, interpMode=RCIM_Cubic, leaveTangent=0, tangentMode=RCTM_User, time=0}
      {arriveTangent=0, interpMode=RCIM_Cubic, leaveTangent=0, tangentMode=RCTM_User, time=20, value=0.92}
      {arriveTangent=0.004, interpMode=RCIM_Cubic, leaveTangent=0.004, time=40}
     ]}}
   volumeCurve {editorCurveData {keys [<at 0 1>]}}
  }
 }
 {
  id=Enveloper_5
  children [Enveloper_4]
  properties {
   bLoop=true
   bLoopIndefinitely=true
   loopEnd=30
   pitchCurve {editorCurveData {keys [{time=0}]}}
   volumeCurve {editorCurveData {keys [
      <at 0 1 RCIM_Cubic>
      {arriveTangent=0, interpMode=RCIM_Cubic, leaveTangent=0, tangentMode=RCTM_User, time=15, value=1.1}
      <at 30 1 RCIM_Cubic>
     ]}}
  }
 }
 {
  id=Enveloper_6
  children [Enveloper_0]
  properties {
   bLoop=true
   bLoopIndefinitely=true
   loopEnd=30
   pitchCurve {editorCurveData {keys [{time=0}]}}
   volumeCurve {editorCurveData {keys [
      <at 0 1 RCIM_Cubic>
      {arriveTangent=0, interpMode=RCIM_Cubic, leaveTangent=0, tangentMode=RCTM_User, time=15, value=1.1}
      <at 30 1 RCIM_Cubic>
     ]}}
  }
 }
 {id=Mixer_0, children [WavePlayer_7,WavePlayer_0], properties {inputVolume [1,1]}}
 {id=Mixer_3, children [WavePlayer_1,WavePlayer_6], properties {inputVolume [1,1]}}
 {id=QualityLevel_1, children [Enveloper_6,Enveloper_6,Enveloper_5,Enveloper_5]}
 {
  id=WavePlayer_0
  properties {bLooping=true, soundWaveAssetPtr=/AmbientAudioContent/Waves/POI/Ambience_Wood_Interior_Windy_With_Bangs_Loop_02_Quad}
 }
 {
  id=WavePlayer_1
  properties {bLooping=true, soundWaveAssetPtr=/AmbientAudioContent/Waves/POI/Ambience_Wood_Interior_Windy_With_Bangs_Loop_02_Stereo}
 }
 {
  id=WavePlayer_6
  properties {bLooping=true, soundWaveAssetPtr=/AmbientAudioContent/Waves/POI/Ambience_DCastle_Wood_Attic_Room_Interior_Loop_Stereo}
 }
 {
  id=WavePlayer_7
  properties {bLooping=true, soundWaveAssetPtr=/AmbientAudioContent/Waves/POI/Ambience_DCastle_Wood_Attic_Room_Interior_Loop_Quad}
 }
]
properties {
 attenuationSettings=/AmbientAudioBase/Attenuations/Fort_Ambient_Att
 bEnableBusSends=false
 bOverrideConcurrency=true
 maxDistance=2097152
 soundClassObject=&Fort_Ambient_Indoors_Athena
 volumeMultiplier=0.4
 concurrencyOverrides {maxCount=1}
 concurrencySet [/AmbientAudioBase/Concurrency/AmbientConc_2Max_Loop_PreventNew_Default_Concurrency]
}
