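# Looping outdoor daytime ambience cue of distant metal rattling in wind, for points of interest.
# A quality-level node routes its first two levels to quad waves and its last two to stereo, each branch picking randomly between two looping variants.
# Nested endlessly looping envelopers modulate pitch over 30 seconds and volume over 40, so the differing periods vary the loop.
# It plays at reduced volume with ambient attenuation and outdoor sound class.
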SoundCue /AmbientAudioContent/SoundCues/POI/Ambience_Metal_Distant_Rattling_In_Wind_Loop_Day_Ext_Cue
root=QualityLevel_0
nodes [
 {
  id=Enveloper_0
  children [Random_0]
  properties {
   bLoop=true
   bLoopIndefinitely=true
   loopEnd=30
   pitchCurve {editorCurveData {keys [{interpMode=RCIM_Cubic, time=0}, <at 15.9468 0.9181 RCIM_Cubic>, {interpMode=RCIM_Cubic, time=30}]}}
   volumeCurve {editorCurveData {keys [<at 0 1>]}}
  }
 }
 {
  id=Enveloper_1
  children [Random_1]
  properties {
   bLoop=true
   bLoopIndefinitely=true
   loopEnd=30
   pitchCurve {editorCurveData {keys [{interpMode=RCIM_Cubic, time=0}, <at 15.9468 0.9181 RCIM_Cubic>, {interpMode=RCIM_Cubic, time=30}]}}
   volumeCurve {editorCurveData {keys [<at 0 1>]}}
  }
 }
 {
  id=Enveloper_3
  children [Enveloper_0]
  properties {
   bLoop=true
   bLoopIndefinitely=true
   loopEnd=40
   pitchCurve {editorCurveData {keys [{time=0}]}}
   volumeCurve {editorCurveData {keys [
      {arriveTangent=-0.01, interpMode=RCIM_Cubic, time=0, value=1}
      <at 10 0.6 RCIM_Cubic>
      {arriveTangent=-0.0105, interpMode=RCIM_Cubic, leaveTangent=-0.0105, tangentMode=RCTM_User, time=20, value=1}
      {arriveTangent=-0.025, interpMode=RCIM_Cubic, leaveTangent=-0.025, time=23.5506, value=0.4}
      {arriveTangent=0.0013, interpMode=RCIM_Cubic, leaveTangent=0.0013, tangentMode=RCTM_User, time=28, value=0.8}
      {arriveTangent=0.0167, interpMode=RCIM_Cubic, leaveTangent=0.0167, time=33, value=0.6}
      <at 40 1>
     ]}}
  }
 }
 {
  id=Enveloper_4
  children [Enveloper_1]
  properties {
   bLoop=true
   bLoopIndefinitely=true
   loopEnd=40
   pitchCurve {editorCurveData {keys [{time=0}]}}
   volumeCurve {editorCurveData {keys [
      {arriveTangent=-0.01, interpMode=RCIM_Cubic, time=0, value=1}
      <at 10 0.6 RCIM_Cubic>
      {arriveTangent=-0.0105, interpMode=RCIM_Cubic, leaveTangent=-0.0105, tangentMode=RCTM_User, time=20, value=1}
      {arriveTangent=-0.025, interpMode=RCIM_Cubic, leaveTangent=-0.025, time=23.5506, value=0.4}
      {arriveTangent=0.0013, interpMode=RCIM_Cubic, leaveTangent=0.0013, tangentMode=RCTM_User, time=28, value=0.8}
      {arriveTangent=0.0167, interpMode=RCIM_Cubic, leaveTangent=0.0167, time=33, value=0.6}
      <at 40 1>
     ]}}
  }
 }
 {id=QualityLevel_0, children [Enveloper_3,Enveloper_3,Enveloper_4,Enveloper_4]}
 {id=Random_0, children [WavePlayer_1,WavePlayer_4], properties {weights [1,1]}}
 {id=Random_1, children [WavePlayer_3,WavePlayer_5], properties {weights [1,1]}}
 {
  id=WavePlayer_1
  properties {bLooping=true, soundWaveAssetPtr=/AmbientAudioContent/Waves/POI/Ambience_Metal_Distant_Rattling_In_Wind_Loop_01_Quad}
 }
 {
  id=WavePlayer_3
  properties {bLooping=true, soundWaveAssetPtr=/AmbientAudioContent/Waves/POI/Ambience_Metal_Distant_Rattling_In_Wind_Loop_01_Stereo}
 }
 {
  id=WavePlayer_4
  properties {bLooping=true, soundWaveAssetPtr=/AmbientAudioContent/Waves/POI/Ambience_Metal_Distant_Rattling_In_Wind_Loop_02_Quad}
 }
 {
  id=WavePlayer_5
  properties {bLooping=true, soundWaveAssetPtr=/AmbientAudioContent/Waves/POI/Ambience_Metal_Distant_Rattling_In_Wind_Loop_02_Stereo}
 }
]
properties {
 attenuationSettings=&Fort_Ambient_Athena_Att
 maxDistance=2097152
 soundClassObject=&Fort_Ambient_Outdoors_Athena
 volumeMultiplier=0.2
 preEffectBusSends [{}]
}
