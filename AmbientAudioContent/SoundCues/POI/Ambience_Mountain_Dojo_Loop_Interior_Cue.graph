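# Looping interior ambience Sound Cue for the Mountain Dojo POI, routed to the Athena indoor ambient sound class.
# A quality-level node plays the quad wave on the first two levels and the stereo wave on the last two.
# Each looping wave feeds two nested, endlessly looping Envelopers: the inner sweeps pitch over a 30-second cycle, while the outer scales pitch down and shapes volume with a shallow curve.
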SoundCue /AmbientAudioContent/SoundCues/POI/Ambience_Mountain_Dojo_Loop_Interior_Cue
root=QualityLevel_0
nodes [
 {
  id=Enveloper_3
  children [WavePlayer_0]
  properties {
   bLoop=true
   bLoopIndefinitely=true
   loopEnd=30
   pitchCurve {editorCurveData {keys [{interpMode=RCIM_Cubic, time=0}, <at 15.9468 0.9181 RCIM_Cubic>, {interpMode=RCIM_Cubic, time=30}]}}
   volumeCurve {editorCurveData {keys [<at 0 1>]}}
  }
 }
 {
  id=Enveloper_5
  children [Enveloper_6]
  properties {
   bLoop=true
   bLoopIndefinitely=true
   loopEnd=30
   pitchCurve {editorCurveData {keys [<at 0 0.8>]}}
   volumeCurve {editorCurveData {keys [<at 0 1 RCIM_Cubic>,<at 50 0.9 RCIM_Cubic>,<at 100 1 RCIM_Cubic>]}}
  }
 }
 {
  id=Enveloper_6
  children [WavePlayer_1]
  properties {
   bLoop=true
   bLoopIndefinitely=true
   loopEnd=30
   pitchCurve {editorCurveData {keys [{interpMode=RCIM_Cubic, time=0}, <at 15.9468 0.9181 RCIM_Cubic>, {interpMode=RCIM_Cubic, time=30}]}}
   volumeCurve {editorCurveData {keys [<at 0 1>]}}
  }
 }
 {
  id=Enveloper_7
  children [Enveloper_3]
  properties {
   bLoop=true
   bLoopIndefinitely=true
   loopEnd=30
   pitchCurve {editorCurveData {keys [<at 0 0.8>]}}
   volumeCurve {editorCurveData {keys [<at 0 1 RCIM_Cubic>,<at 50 0.9 RCIM_Cubic>,<at 100 1 RCIM_Cubic>]}}
  }
 }
 {id=QualityLevel_0, children [Enveloper_7,Enveloper_7,Enveloper_5,Enveloper_5]}
 {id=WavePlayer_0, properties {bLooping=true, soundWaveAssetPtr=/AmbientAudioContent/Waves/POI/Ambience_Mountain_Dojo_Loop_Int_Quad}}
 {id=WavePlayer_1, properties {bLooping=true, soundWaveAssetPtr=/AmbientAudioContent/Waves/POI/Ambience_Mountain_Dojo_Loop_Int_Stereo}}
]
properties {maxDistance=2097152, soundClassObject=&Fort_Ambient_Indoors_Athena, volumeMultiplier=1.15}
