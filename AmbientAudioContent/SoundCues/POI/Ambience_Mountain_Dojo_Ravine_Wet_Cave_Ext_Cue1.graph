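# Ambient sound cue for the exterior of a wet cave in the Mountain Dojo Ravine POI, slightly pitched up and using Athena ambient attenuation.
# One looping cave water wave feeds two identical Envelopers, whose 20-second volume curves dip to half and return.
# A QualityLevel root picks between two Mixers, each with one empty input and one Enveloper at the same reduced volume, so all quality branches are effectively alike.
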SoundCue /AmbientAudioContent/SoundCues/POI/Ambience_Mountain_Dojo_Ravine_Wet_Cave_Ext_Cue1
root=QualityLevel_0
nodes [
 {
  id=Enveloper_0
  children [WavePlayer_4]
  properties {
   loopEnd=20
   pitchCurve {editorCurveData {keys [{time=0}]}}
   volumeCurve {editorCurveData {keys [<at 0 1 RCIM_Cubic>,<at 10 0.5 RCIM_Cubic>,<at 20 1>]}}
  }
 }
 {
  id=Enveloper_1
  children [WavePlayer_4]
  properties {
   loopEnd=20
   pitchCurve {editorCurveData {keys [{time=0}]}}
   volumeCurve {editorCurveData {keys [<at 0 1 RCIM_Cubic>,<at 10 0.5 RCIM_Cubic>,<at 20 1>]}}
  }
 }
 {id=Mixer_0, children [Enveloper_0,null], properties {inputVolume [0.6,0.75]}}
 {id=Mixer_1, children [null,Enveloper_1], properties {inputVolume [0.75,0.6]}}
 {id=QualityLevel_0, children [Mixer_0,Mixer_0,Mixer_1,Mixer_1]}
 {id=WavePlayer_4, properties {bLooping=true, soundWaveAssetPtr=/AmbientAudioContent/Waves/Water/Spot_Cattus_Cave_Water_Loop}}
]
properties {
 attenuationSettings=&Fort_Ambient_Athena_Att
 maxDistance=2097152
 pitchMultiplier=1.1
 soundClassObject=&Fort_Ambient_Athena
 volumeMultiplier=1
 concurrencyOverrides {maxCount=1, resolutionRule=PreventNew}
}
