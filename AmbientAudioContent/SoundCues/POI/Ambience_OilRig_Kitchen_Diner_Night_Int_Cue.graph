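# Indoor night ambience cue for the oil rig kitchen or diner POI, on the indoor ambient sound class.
# A quality switch gives the first two levels a mix of looping quad restaurant room tone and a quieter, pitched-down oil rig interior loop.
# The last two levels play only a looping stereo version of the room tone.
# Overridden concurrency prevents new instances once the limit is reached.
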SoundCue /AmbientAudioContent/SoundCues/POI/Ambience_OilRig_Kitchen_Diner_Night_Int_Cue
root=QualityLevel_0
nodes [
 {id=Mixer_1, children [Modulator_0,WavePlayer_0], properties {inputVolume [0.15,1]}}
 {id=Mixer_3, children [WavePlayer_1], properties {inputVolume [1]}}
 {id=Modulator_0, children [WavePlayer_4], properties {pitchMax=0.7, pitchMin=0.7}}
 {id=QualityLevel_0, children [Mixer_1,Mixer_1,Mixer_3,Mixer_3]}
 {id=WavePlayer_0, properties {bLooping=true, soundWaveAssetPtr=/AmbientAudioContent/Waves/POI/Ambience_Resturant_Medium_Room_Tone_Quad}}
 {id=WavePlayer_1, properties {bLooping=true, soundWaveAssetPtr=/AmbientAudioContent/Waves/POI/Ambience_Resturant_Medium_Room_Tone_Stereo}}
 {id=WavePlayer_4, properties {bLooping=true, soundWaveAssetPtr=/AmbientAudioContent/Waves/POI/Ambience_OilRig_Interior_Sml_Quad}}
]
properties {
 attenuationSettings=&Fort_Ambient_Athena_Att
 bOverrideConcurrency=true
 maxDistance=2097152
 soundClassObject=&Fort_Ambient_Indoors_Athena
 volumeMultiplier=0.8
 attenuationOverrides {bAttenuate=false, bEnableReverbSend=false, bSpatialize=false}
 concurrencyOverrides {resolutionRule=PreventNew}
 preEffectBusSends [{sendLevel=0}]
}
