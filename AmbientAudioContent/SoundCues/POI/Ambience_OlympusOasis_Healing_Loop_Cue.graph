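# Looping, non-spatialized ambience cue for the Olympus Oasis healing POI, limited to one concurrent instance with no distance volume attenuation.
# A quality switch picks between quad and stereo water-tunnel beds, each mixed with a health-globe loop.
# Nested looping envelopers add irregular volume swells to the globe and slow pitch and volume drift to each mix.
# Mixer weights are reversed between branches: water dominates in quad, the globe in stereo.
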SoundCue /AmbientAudioContent/SoundCues/POI/Ambience_OlympusOasis_Healing_Loop_Cue
root=QualityLevel_2
nodes [
 {
  id=Enveloper_1
  children [Mixer_4]
  properties {
   bLoop=true
   bLoopIndefinitely=true
   loopEnd=30
   pitchCurve {editorCurveData {keys [
      {arriveTangent=-0.0013, interpMode=RCIM_Cubic, leaveTangent=-0.0013, time=0}
      <at 15 0.98 RCIM_Cubic>
      {arriveTangent=0.0013, interpMode=RCIM_Cubic, leaveTangent=0.0013, time=30}
     ]}}
   volumeCurve {editorCurveData {keys [<at 0 1>]}}
  }
 }
 {
  id=Enveloper_10
  children [Enveloper_9]
  properties {
   bLoop=true
   bLoopIndefinitely=true
   loopEnd=30
   pitchCurve {editorCurveData {keys [{time=0}]}}
   volumeCurve {editorCurveData {keys [
      {arriveTangent=-0.01, interpMode=RCIM_Cubic, time=0, value=1}
      <at 10 0.5 RCIM_Cubic>
      {interpMode=RCIM_Cubic, leaveTangent=0.01, time=20, value=1}
     ]}}
  }
 }
 {
  id=Enveloper_2
  children [Enveloper_1]
  properties {
   bLoop=true
   bLoopIndefinitely=true
   loopEnd=20
   pitchCurve {editorCurveData {keys [{time=0}]}}
   volumeCurve {editorCurveData {keys [
      {arriveTangent=-0.01, interpMode=RCIM_Cubic, leaveTangent=-0.01, time=0, value=1}
      <at 10 0.9 RCIM_Cubic>
      {arriveTangent=0.01, interpMode=RCIM_Cubic, leaveTangent=0.01, time=20, value=1}
     ]}}
  }
 }
 {
  id=Enveloper_5
  children [Mixer_5]
  properties {
   bLoop=true
   bLoopIndefinitely=true
   loopEnd=30
   pitchCurve {editorCurveData {keys [
      {arriveTangent=-0.0013, interpMode=RCIM_Cubic, leaveTangent=-0.0013, time=0}
      <at 15 0.98 RCIM_Cubic>
      {arriveTangent=0.0013, interpMode=RCIM_Cubic, leaveTangent=0.0013, time=30}
     ]}}
   volumeCurve {editorCurveData {keys [<at 0 1>]}}
  }
 }
 {
  id=Enveloper_6
  children [Enveloper_5]
  properties {
   bLoop=true
   bLoopIndefinitely=true
   loopEnd=20
   pitchCurve {editorCurveData {keys [{time=0}]}}
   volumeCurve {editorCurveData {keys [
      {arriveTangent=-0.01, interpMode=RCIM_Cubic, leaveTangent=-0.01, time=0, value=1}
      <at 10 0.9 RCIM_Cubic>
      {arriveTangent=0.01, interpMode=RCIM_Cubic, leaveTangent=0.01, time=20, value=1}
     ]}}
  }
 }
 {
  id=Enveloper_7
  children [WavePlayer_0]
  properties {
   bLoop=true
   bLoopIndefinitely=true
   loopEnd=60
   pitchCurve {editorCurveData {keys [{time=0}]}}
   volumeCurve {editorCurveData {keys [
      {arriveTangent=-0.01, interpMode=RCIM_Cubic, time=0, value=1}
      {arriveTangent=-0.0035, interpMode=RCIM_Cubic, leaveTangent=-0.0035, tangentMode=RCTM_User, time=4.9592, value=0.6678}
      {arriveTangent=0.0016, interpMode=RCIM_Cubic, leaveTangent=0.0016, tangentMode=RCTM_User, time=7.5626, value=0.9613}
      {arriveTangent=0.0008, interpMode=RCIM_Cubic, leaveTangent=0.0008, tangentMode=RCTM_User, time=11.5241, value=0.6812}
      {arriveTangent=0.0016, interpMode=RCIM_Cubic, leaveTangent=0.0016, tangentMode=RCTM_User, time=13.3939, value=0.6668}
      {arriveTangent=-0.0011, interpMode=RCIM_Cubic, leaveTangent=-0.0011, tangentMode=RCTM_User, time=19.0032, value=0.6718}
      {arriveTangent=0.0002, interpMode=RCIM_Cubic, leaveTangent=0.0002, time=24.105, value=0.9798}
      {arriveTangent=0.0016, interpMode=RCIM_Cubic, leaveTangent=0.0016, tangentMode=RCTM_User, time=28.4504, value=0.674}
      {arriveTangent=0.0009, interpMode=RCIM_Cubic, leaveTangent=0.0009, tangentMode=RCTM_User, time=35.3391, value=0.672}
      {arriveTangent=-0.0008, interpMode=RCIM_Cubic, leaveTangent=-0.0008, time=38.0791, value=0.9798}
      <at 42.4245 0.6667>
      {arriveTangent=0.0024, interpMode=RCIM_Cubic, leaveTangent=0.0024, tangentMode=RCTM_User, time=44.9832, value=0.6718}
      {arriveTangent=-0.0006, interpMode=RCIM_Cubic, leaveTangent=-0.0006, time=48.1071, value=0.9613}
      {arriveTangent=-0.0056, interpMode=RCIM_Cubic, leaveTangent=-0.0056, tangentMode=RCTM_User, time=53.5507, value=0.6668}
      {arriveTangent=-0.0003, interpMode=RCIM_Cubic, leaveTangent=-0.0003, tangentMode=RCTM_User, time=58.17, value=0.6665}
      <at 60 1 RCIM_Cubic>
     ]}}
  }
 }
 {
  id=Enveloper_8
  children [Enveloper_7]
  properties {
   bLoop=true
   bLoopIndefinitely=true
   loopEnd=30
   pitchCurve {editorCurveData {keys [{time=0}]}}
   volumeCurve {editorCurveData {keys [
      {arriveTangent=-0.01, interpMode=RCIM_Cubic, time=0, value=1}
      <at 10 0.5 RCIM_Cubic>
      {interpMode=RCIM_Cubic, leaveTangent=0.01, time=20, value=1}
     ]}}
  }
 }
 {
  id=Enveloper_9
  children [WavePlayer_7]
  properties {
   bLoop=true
   bLoopIndefinitely=true
   loopEnd=60
   pitchCurve {editorCurveData {keys [{time=0}]}}
   volumeCurve {editorCurveData {keys [
      {arriveTangent=-0.01, interpMode=RCIM_Cubic, time=0, value=1}
      {arriveTangent=-0.0035, interpMode=RCIM_Cubic, leaveTangent=-0.0035, tangentMode=RCTM_User, time=4.9592, value=0.6678}
      {arriveTangent=0.0016, interpMode=RCIM_Cubic, leaveTangent=0.0016, tangentMode=RCTM_User, time=7.5626, value=0.9613}
      {arriveTangent=0.0008, interpMode=RCIM_Cubic, leaveTangent=0.0008, tangentMode=RCTM_User, time=11.5241, value=0.6812}
      {arriveTangent=0.0016, interpMode=RCIM_Cubic, leaveTangent=0.0016, tangentMode=RCTM_User, time=13.3939, value=0.6668}
      {arriveTangent=-0.0011, interpMode=RCIM_Cubic, leaveTangent=-0.0011, tangentMode=RCTM_User, time=19.0032, value=0.6718}
      {arriveTangent=0.0002, interpMode=RCIM_Cubic, leaveTangent=0.0002, time=24.105, value=0.9798}
      {arriveTangent=0.0016, interpMode=RCIM_Cubic, leaveTangent=0.0016, tangentMode=RCTM_User, time=28.4504, value=0.674}
      {arriveTangent=0.0009, interpMode=RCIM_Cubic, leaveTangent=0.0009, tangentMode=RCTM_User, time=35.3391, value=0.672}
      {arriveTangent=-0.0008, interpMode=RCIM_Cubic, leaveTangent=-0.0008, time=38.0791, value=0.9798}
      <at 42.4245 0.6667>
      {arriveTangent=0.0024, interpMode=RCIM_Cubic, leaveTangent=0.0024, tangentMode=RCTM_User, time=44.9832, value=0.6718}
      {arriveTangent=-0.0006, interpMode=RCIM_Cubic, leaveTangent=-0.0006, time=48.1071, value=0.9613}
      {arriveTangent=-0.0056, interpMode=RCIM_Cubic, leaveTangent=-0.0056, tangentMode=RCTM_User, time=53.5507, value=0.6668}
      {arriveTangent=-0.0003, interpMode=RCIM_Cubic, leaveTangent=-0.0003, tangentMode=RCTM_User, time=58.17, value=0.6665}
      <at 60 1 RCIM_Cubic>
     ]}}
  }
 }
 {id=Mixer_4, children [Enveloper_8,WavePlayer_1], properties {inputVolume [0.2,1]}}
 {id=Mixer_5, children [WavePlayer_6,Enveloper_10], properties {inputVolume [0.2,1]}}
 {id=QualityLevel_2, children [Enveloper_2,Enveloper_2,Enveloper_6,Enveloper_6]}
 {id=WavePlayer_0, properties {bLooping=true, soundWaveAssetPtr=/AmbientAudioContent/Waves/SpottedLoops/Spot_STW_HealthGlobe_Loop_01}}
 {
  id=WavePlayer_1
  properties {bLooping=true, soundWaveAssetPtr=/AmbientAudioContent/Waves/POI/Ambience_Hollow_Base_Water_Tunnel_Day_Loop_Quad}
 }
 {
  id=WavePlayer_6
  properties {bLooping=true, soundWaveAssetPtr=/AmbientAudioContent/Waves/POI/Ambience_Hollow_Base_Water_Tunnel_Day_Loop_Stereo}
 }
 {id=WavePlayer_7, properties {bLooping=true, soundWaveAssetPtr=&Spot_STW_HealthGlobe_Loop_01}}
]
properties {
 bOverrideAttenuation=true
 bOverrideConcurrency=true
 maxDistance=2097152
 soundClassObject=&Fort_Ambient_Indoors_Athena
 volumeMultiplier=0.501
 attenuationOverrides {
  bAttenuate=false
  bAttenuateWithLPF=true
  bEnableFocusInterpolation=true
  bEnableLogFrequencyScaling=true
  bSpatialize=false
  falloffDistance=2000
  focusAttackInterpSpeed=0.1
  focusReleaseInterpSpeed=0.1
  hPFFrequencyAtMax=500
  hPFFrequencyAtMin=500
  lPFRadiusMax=0
  lPFRadiusMin=0
  nonFocusAzimuth=90
  nonFocusVolumeAttenuation=0.5
  nonSpatializedRadiusStart=500
  stereoSpread=500
  attenuationShapeExtents {x=500}
 }
 concurrencyOverrides {maxCount=1, voiceStealReleaseTime=2}
}
