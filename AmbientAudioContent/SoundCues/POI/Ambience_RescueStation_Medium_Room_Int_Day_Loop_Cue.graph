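# Looping, non-spatialized daytime interior ambience cue for a medium room at a rescue station POI, in the indoor ambient sound class.
# It layers a faint, pitched-down exterior station loop under an airy residential room tone.
# A quality switch feeds its first two levels quad waves plus a boosted cave metal-creaking layer, and its last two stereo versions without the creaking.
# Concurrency is overridden to prevent new instances at the limit.
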SoundCue /AmbientAudioContent/SoundCues/POI/Ambience_RescueStation_Medium_Room_Int_Day_Loop_Cue
root=QualityLevel_0
nodes [
 {id=Mixer_1, children [Modulator_0,WavePlayer_2,WavePlayer_8], properties {inputVolume [0.1,0.6,1.5]}}
 {id=Mixer_3, children [Modulator_1,WavePlayer_3], properties {inputVolume [0.1,1]}}
 {id=Modulator_0, children [WavePlayer_0], properties {pitchMax=0.8, pitchMin=0.8}}
 {id=Modulator_1, children [WavePlayer_1], properties {pitchMax=0.8, pitchMin=0.8}}
 {id=QualityLevel_0, children [Mixer_1,Mixer_1,Mixer_3,Mixer_3]}
 {
  id=WavePlayer_0
  properties {bLooping=true, soundWaveAssetPtr=/AmbientAudioContent/Waves/POI/Ambience_MountainRescueStation_Day_Loop_Ext_Quad}
 }
 {
  id=WavePlayer_1
  properties {bLooping=true, soundWaveAssetPtr=/AmbientAudioContent/Waves/POI/Ambience_MountainRescueStation_Day_Loop_Ext_Stereo}
 }
 {
  id=WavePlayer_2
  properties {bLooping=true, soundWaveAssetPtr=/AmbientAudioContent/Waves/POI/Ambience_Residential_Medium_Room_Tone_Airy_Quad}
 }
 {
  id=WavePlayer_3
  properties {bLooping=true, soundWaveAssetPtr=/AmbientAudioContent/Waves/POI/Ambience_Residential_Medium_Room_Tone_Airy_Stereo}
 }
 {
  id=WavePlayer_8
  properties {bLooping=true, soundWaveAssetPtr=/AmbientAudioContent/Waves/POI/Ambience_Underground_Lair_Cave_Metal_Creaking_Quad_Loop}
 }
]
properties {
 bOverrideConcurrency=true
 maxDistance=2097152
 soundClassObject=&Fort_Ambient_Indoors_Athena
 volumeMultiplier=0.8
 attenuationOverrides {bAttenuate=false, bEnableReverbSend=false, bSpatialize=false}
 concurrencyOverrides {resolutionRule=PreventNew}
}
