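# Looping outdoor ambient cue that, by its name, adds an exterior cloth layer at the shipwreck POI, reusing blimp loading bay wind and cloth-flap loops.
# A quality node routes its first two levels to the quad wave and last two to the stereo one.
# Each looping wave passes through identical nested envelopers: an inner 30-second pitch curve and an outer 60-second volume curve with irregular swells and dips, both repeating indefinitely to vary the loop.
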SoundCue /AmbientAudioContent/SoundCues/POI/Ambience_Shipwreck_Cloth_Additive_Layer_Loop_Ext
root=QualityLevel_0
nodes [
 {
  id=Enveloper_0
  children [WavePlayer_1]
  properties {
   bLoop=true
   bLoopIndefinitely=true
   loopEnd=30
   pitchCurve {editorCurveData {keys [{interpMode=RCIM_Cubic, time=0}, <at 15.9468 0.9181 RCIM_Cubic>, {interpMode=RCIM_Cubic, time=30}]}}
   volumeCurve {editorCurveData {keys [<at 0 1>]}}
  }
 }
 {
  id=Enveloper_1
  children [WavePlayer_3]
  properties {
   bLoop=true
   bLoopIndefinitely=true
   loopEnd=30
   pitchCurve {editorCurveData {keys [{interpMode=RCIM_Cubic, time=0}, <at 15.9468 0.9181 RCIM_Cubic>, {interpMode=RCIM_Cubic, time=30}]}}
   volumeCurve {editorCurveData {keys [<at 0 1>]}}
  }
 }
 {
  id=Enveloper_3
  children [Enveloper_0]
  properties {
   bLoop=true
   bLoopIndefinitely=true
   loopEnd=60
   pitchCurve {editorCurveData {keys [{time=0}]}}
   volumeCurve {editorCurveData {keys [
      {arriveTangent=-0.01, interpMode=RCIM_Cubic, time=0, value=1}
      {arriveTangent=0.0003, interpMode=RCIM_Cubic, leaveTangent=0.0003, time=10, value=0.6}
      {arriveTangent=0.0005, interpMode=RCIM_Cubic, leaveTangent=0.0005, tangentMode=RCTM_User, time=19.3202, value=1.0053}
      {arriveTangent=-0.0011, interpMode=RCIM_Cubic, leaveTangent=-0.0011, time=29.0811, value=0.2917}
      {arriveTangent=-0.0002, interpMode=RCIM_Cubic, leaveTangent=-0.0002, tangentMode=RCTM_User, time=35.0705, value=0.9887}
      {arriveTangent=0.0005, interpMode=RCIM_Cubic, leaveTangent=0.0005, time=46.3307, value=0.1}
      {interpMode=RCIM_Cubic, leaveTangent=0.01, time=60, value=1}
     ]}}
  }
 }
 {
  id=Enveloper_4
  children [Enveloper_1]
  properties {
   bLoop=true
   bLoopIndefinitely=true
   loopEnd=60
   pitchCurve {editorCurveData {keys [{time=0}]}}
   volumeCurve {editorCurveData {keys [
      {arriveTangent=-0.01, interpMode=RCIM_Cubic, time=0, value=1}
      {arriveTangent=0.0003, interpMode=RCIM_Cubic, leaveTangent=0.0003, time=10, value=0.6}
      {arriveTangent=0.0005, interpMode=RCIM_Cubic, leaveTangent=0.0005, tangentMode=RCTM_User, time=19.3202, value=1.0053}
      {arriveTangent=-0.0011, interpMode=RCIM_Cubic, leaveTangent=-0.0011, time=29.0811, value=0.2917}
      {arriveTangent=-0.0002, interpMode=RCIM_Cubic, leaveTangent=-0.0002, tangentMode=RCTM_User, time=35.0705, value=0.9887}
      {arriveTangent=0.0005, interpMode=RCIM_Cubic, leaveTangent=0.0005, time=46.3307, value=0.1}
      {interpMode=RCIM_Cubic, leaveTangent=0.01, time=60, value=1}
     ]}}
  }
 }
 {id=QualityLevel_0, children [Enveloper_3,Enveloper_3,Enveloper_4,Enveloper_4]}
 {
  id=WavePlayer_1
  properties {bLooping=true, soundWaveAssetPtr=/AmbientAudioContent/Waves/POI/Ambience_Blimp_Interior_Loading_Bay_Wind_Cloth_Flaps_Loop_Quad}
 }
 {
  id=WavePlayer_3
  properties {bLooping=true, soundWaveAssetPtr=/AmbientAudioContent/Waves/POI/Ambience_Blimp_Interior_Loading_Bay_Wind_Cloth_Flaps_Loop_Stereo}
 }
]
properties {
 attenuationSettings=&Fort_Ambient_Athena_Att
 maxDistance=2097152
 soundClassObject=&Fort_Ambient_Outdoors_Athena
 volumeMultiplier=0.301
 preEffectBusSends [{sendLevel=0}]
}
