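# Interior ambience Sound Cue for the SprocketPoppy02 main POI, using the Athena indoor ambient sound class and ambient attenuation.
# A Quality Level root maps its first two slots to a looping quad wave and its last two to a looping stereo wave.
# Each wave runs through its own Enveloper, which loops indefinitely over a gentle cubic volume dip to add slow level movement.
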SoundCue /AmbientAudioContent/SoundCues/POI/Ambience_SprocketPoppy02_Main_Int_Cue
root=QualityLevel_1
nodes [
 {
  id=Enveloper_0
  children [WavePlayer_3]
  properties {
   bLoop=true
   bLoopIndefinitely=true
   loopEnd=20
   pitchCurve {editorCurveData {keys [{time=0}]}}
   volumeCurve {editorCurveData {keys [
      {arriveTangent=-0.01, interpMode=RCIM_Cubic, leaveTangent=-0.01, time=0, value=1}
      <at 10 0.9 RCIM_Cubic>
      {arriveTangent=0.01, interpMode=RCIM_Cubic, leaveTangent=0.01, time=20, value=1}
     ]}}
  }
 }
 {
  id=Enveloper_3
  children [WavePlayer_2]
  properties {
   bLoop=true
   bLoopIndefinitely=true
   loopEnd=20
   pitchCurve {editorCurveData {keys [{time=0}]}}
   volumeCurve {editorCurveData {keys [
      {arriveTangent=-0.01, interpMode=RCIM_Cubic, leaveTangent=-0.01, time=0, value=1}
      <at 10 0.9 RCIM_Cubic>
      {arriveTangent=0.01, interpMode=RCIM_Cubic, leaveTangent=0.01, time=20, value=1}
     ]}}
  }
 }
 {id=QualityLevel_1, children [Enveloper_3,Enveloper_3,Enveloper_0,Enveloper_0]}
 {id=WavePlayer_2, properties {bLooping=true, soundWaveAssetPtr=/AmbientAudioContent/Waves/POI/Ambience_SprocketPoppy02_Main_Quad_Loop}}
 {id=WavePlayer_3, properties {bLooping=true, soundWaveAssetPtr=/AmbientAudioContent/Waves/POI/Ambience_SprocketPoppy02_Main_Stereo_Loop}}
]
properties {
 attenuationSettings=&Fort_Ambient_Athena_Att
 maxDistance=2097152
 soundClassObject=&Fort_Ambient_Indoors_Athena
 volumeMultiplier=1.5
 preEffectBusSends [{}]
}
