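# Looping, non-spatialized interior ambience cue for a large tech room POI in the starter zone.
# A quality-level switch picks a quad or stereo mix of room tone, airy computer-room loop and a slowly volume-modulated UFO tech loop.
# Nested looping envelopers add subtle pitch and volume drift, under an outer fade-in.
# The cue sends to reverb and, despite its interior name, the exterior ambient bus, with concurrency preventing new instances.
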SoundCue /AmbientAudioContent/SoundCues/POI/Ambience_Starter_Zone_Large_Tech_Room_Int_Cue
root=Enveloper_7
nodes [
 {
  id=Enveloper_0
  children [QualityLevel_0]
  properties {
   bLoop=true
   bLoopIndefinitely=true
   loopEnd=30
   pitchCurve {editorCurveData {keys [
      {arriveTangent=-0.0013, interpMode=RCIM_Cubic, leaveTangent=-0.0013, time=0}
      <at 15 0.98 RCIM_Cubic>
      {arriveTangent=0.0013, interpMode=RCIM_Cubic, leaveTangent=0.0013, time=30}
     ]}}
   volumeCurve {editorCurveData {keys [<at 0 1>]}}
  }
 }
 {
  id=Enveloper_1
  children [WavePlayer_0]
  properties {
   bLoop=true
   bLoopIndefinitely=true
   loopEnd=60
   pitchCurve {editorCurveData {keys [{time=0}]}}
   volumeCurve {editorCurveData {keys [
      {arriveTangent=-0.01, interpMode=RCIM_Cubic, time=0, value=1}
      <at 30 0.8 RCIM_Cubic>
      {interpMode=RCIM_Cubic, leaveTangent=0.01, time=60, value=1}
     ]}}
  }
 }
 {
  id=Enveloper_3
  children [Enveloper_0]
  properties {
   bLoop=true
   bLoopIndefinitely=true
   loopEnd=20
   pitchCurve {editorCurveData {keys [{time=0}]}}
   volumeCurve {editorCurveData {keys [
      {arriveTangent=-0.01, interpMode=RCIM_Cubic, leaveTangent=-0.01, time=0, value=1}
      <at 10 0.9 RCIM_Cubic>
      {arriveTangent=0.01, interpMode=RCIM_Cubic, leaveTangent=0.01, time=20, value=1}
     ]}}
  }
 }
 {
  id=Enveloper_4
  children [WavePlayer_1]
  properties {
   bLoop=true
   bLoopIndefinitely=true
   loopEnd=60
   pitchCurve {editorCurveData {keys [{time=0}]}}
   volumeCurve {editorCurveData {keys [
      {arriveTangent=-0.01, interpMode=RCIM_Cubic, time=0, value=1}
      <at 30 0.8 RCIM_Cubic>
      {interpMode=RCIM_Cubic, leaveTangent=0.01, time=60, value=1}
     ]}}
  }
 }
 {
  id=Enveloper_7
  children [Enveloper_3]
  properties {pitchCurve {editorCurveData {keys [{time=0}]}}, volumeCurve {editorCurveData {keys [<at 0 0>,<at 2 1>]}}}
 }
 {id=Mixer_0, children [WavePlayer_2,Enveloper_1,WavePlayer_4], properties {inputVolume [4,1.5,0.5]}}
 {id=Mixer_1, children [WavePlayer_3,Enveloper_4,WavePlayer_5], properties {inputVolume [4,1.5,0.5]}}
 {id=QualityLevel_0, children [Mixer_0,Mixer_0,Mixer_1,Mixer_1]}
 {id=WavePlayer_0, properties {bLooping=true, soundWaveAssetPtr=/AmbientAudioContent/Waves/SpottedLoops/Spot_UFO_Crashsite_Tech_Loop_02}}
 {id=WavePlayer_1, properties {bLooping=true, soundWaveAssetPtr=&Spot_UFO_Crashsite_Tech_Loop_02}}
 {id=WavePlayer_2, properties {bLooping=true, soundWaveAssetPtr=/AmbientAudioContent/Waves/POI/Ambience_Technology_Large_Room_Tone_Quad}}
 {id=WavePlayer_3, properties {bLooping=true, soundWaveAssetPtr=/AmbientAudioContent/Waves/POI/Ambience_Technology_Large_Room_Tone_Stereo}}
 {
  id=WavePlayer_4
  properties {bLooping=true, soundWaveAssetPtr=/AmbientAudioContent/Waves/POI/Ambience_HydroBottle_Lab_ComputerRoom_Loop_Airy_Quad}
 }
 {
  id=WavePlayer_5
  properties {bLooping=true, soundWaveAssetPtr=/AmbientAudioContent/Waves/POI/Ambience_HydroBottle_Lab_ComputerRoom_Loop_Airy_Stereo}
 }
]
properties {
 bOverrideConcurrency=true
 maxDistance=2097152
 soundClassObject=&Fort_Ambient_Athena
 volumeMultiplier=0.2
 attenuationOverrides {bAttenuate=false, bEnableReverbSend=false, bSpatialize=false}
 busSends [{audioBus=null, minSendLevel=1, sendLevel=1, soundSourceBus=&AmbientAudio_Exterior_SourceBus}]
 concurrencyOverrides {resolutionRule=PreventNew}
 soundSubmixSends [{maxSendLevel=0.4, minSendLevel=0.4, soundSubmix=&Fortnite_Reverb_Default}]
}
