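# Looping indoor ambience cue for a water tunnel point of interest, using Fortnite Athena ambient attenuation and indoor sound class.
# A quality-level root sends its first two slots to a quad night-loop wave and its last two to the stereo version.
# Each looping wave passes through two nested, endlessly looping envelopers: a 30-second pitch curve inside a 60-second volume swell, adding slow variation to the bed.
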SoundCue /AmbientAudioContent/SoundCues/POI/Ambience_Tunnel_Water_Loop_Cue
root=QualityLevel_0
nodes [
 {
  id=Enveloper_0
  children [WavePlayer_0]
  properties {
   bLoop=true
   bLoopIndefinitely=true
   loopEnd=30
   pitchCurve {editorCurveData {keys [{interpMode=RCIM_Cubic, time=0}, <at 15.9468 0.9181 RCIM_Cubic>, {interpMode=RCIM_Cubic, time=30}]}}
   volumeCurve {editorCurveData {keys [<at 0 1>]}}
  }
 }
 {
  id=Enveloper_1
  children [Enveloper_0]
  properties {
   bLoop=true
   bLoopIndefinitely=true
   loopEnd=60
   pitchCurve {editorCurveData {keys [{time=0}]}}
   volumeCurve {editorCurveData {keys [
      {arriveTangent=-0.01, interpMode=RCIM_Cubic, time=0, value=1}
      {arriveTangent=0.0006, interpMode=RCIM_Cubic, leaveTangent=0.0006, tangentMode=RCTM_User, time=15, value=0.8}
      {arriveTangent=-0.0006, interpMode=RCIM_Cubic, leaveTangent=-0.0006, tangentMode=RCTM_User, time=25, value=1}
      {arriveTangent=0.0002, interpMode=RCIM_Cubic, leaveTangent=0.0002, tangentMode=RCTM_User, time=45, value=0.6}
      <at 60 1>
     ]}}
  }
 }
 {
  id=Enveloper_4
  children [Enveloper_5]
  properties {
   bLoop=true
   bLoopIndefinitely=true
   loopEnd=60
   pitchCurve {editorCurveData {keys [{time=0}]}}
   volumeCurve {editorCurveData {keys [
      {arriveTangent=-0.01, interpMode=RCIM_Cubic, time=0, value=1}
      {arriveTangent=0.0006, interpMode=RCIM_Cubic, leaveTangent=0.0006, tangentMode=RCTM_User, time=15, value=0.8}
      {arriveTangent=-0.0006, interpMode=RCIM_Cubic, leaveTangent=-0.0006, tangentMode=RCTM_User, time=25, value=1}
      {arriveTangent=0.0002, interpMode=RCIM_Cubic, leaveTangent=0.0002, tangentMode=RCTM_User, time=45, value=0.6}
      <at 60 1>
     ]}}
  }
 }
 {
  id=Enveloper_5
  children [WavePlayer_1]
  properties {
   bLoop=true
   bLoopIndefinitely=true
   loopEnd=30
   pitchCurve {editorCurveData {keys [{interpMode=RCIM_Cubic, time=0}, <at 15.9468 0.9181 RCIM_Cubic>, {interpMode=RCIM_Cubic, time=30}]}}
   volumeCurve {editorCurveData {keys [<at 0 1>]}}
  }
 }
 {id=QualityLevel_0, children [Enveloper_1,Enveloper_1,Enveloper_4,Enveloper_4]}
 {
  id=WavePlayer_0
  properties {bLooping=true, soundWaveAssetPtr=/AmbientAudioContent/Waves/POI/Ambience_Hollow_Base_Water_Tunnel_Night_Loop_Quad}
 }
 {
  id=WavePlayer_1
  properties {bLooping=true, soundWaveAssetPtr=/AmbientAudioContent/Waves/POI/Ambience_Hollow_Base_Water_Tunnel_Night_Loop_Stereo}
 }
]
properties {
 attenuationSettings=&Fort_Ambient_Athena_Att
 maxDistance=2097152
 pitchMultiplier=1.1
 soundClassObject=&Fort_Ambient_Indoors_Athena
 volumeMultiplier=1
}
