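# Looping warehouse ambience SoundCue for points of interest, layering brick and metal interior beds on the Athena ambient class.
# A quality switch feeds quad waves to its first two slots and stereo waves to the last two.
# Each branch mixes a quieter pitched-down bed with a louder one, brick lowered in quad but metal in stereo.
# Nested looping envelopers add slow pitch and volume drift on mismatched cycle lengths.
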SoundCue /AmbientAudioContent/SoundCues/POI/Ambience_Warehouse_Generic_Metal_Stone_Loop_IntExt_Cue
root=QualityLevel_0
nodes [
 {
  id=Enveloper_0
  children [Mixer_0]
  properties {
   bLoop=true
   bLoopIndefinitely=true
   loopEnd=30
   pitchCurve {editorCurveData {keys [{interpMode=RCIM_Cubic, time=0}, <at 15.9468 0.9181 RCIM_Cubic>, {interpMode=RCIM_Cubic, time=30}]}}
   volumeCurve {editorCurveData {keys [<at 0 1>]}}
  }
 }
 {
  id=Enveloper_1
  children [Mixer_1]
  properties {
   bLoop=true
   bLoopIndefinitely=true
   loopEnd=30
   pitchCurve {editorCurveData {keys [{interpMode=RCIM_Cubic, time=0}, <at 15.9468 0.9181 RCIM_Cubic>, {interpMode=RCIM_Cubic, time=30}]}}
   volumeCurve {editorCurveData {keys [<at 0 1>]}}
  }
 }
 {
  id=Enveloper_2
  children [Enveloper_1]
  properties {
   bLoop=true
   bLoopIndefinitely=true
   loopEnd=20
   pitchCurve {editorCurveData {keys [{time=0}]}}
   volumeCurve {editorCurveData {keys [
      {arriveTangent=-0.01, interpMode=RCIM_Cubic, leaveTangent=-0.01, time=0, value=1}
      <at 10 0.9 RCIM_Cubic>
      {arriveTangent=0.01, interpMode=RCIM_Cubic, leaveTangent=0.01, time=20, value=1}
     ]}}
  }
 }
 {
  id=Enveloper_3
  children [Enveloper_0]
  properties {
   bLoop=true
   bLoopIndefinitely=true
   loopEnd=20
   pitchCurve {editorCurveData {keys [{time=0}]}}
   volumeCurve {editorCurveData {keys [
      {arriveTangent=-0.01, interpMode=RCIM_Cubic, leaveTangent=-0.01, time=0, value=1}
      <at 10 0.9 RCIM_Cubic>
      {arriveTangent=0.01, interpMode=RCIM_Cubic, leaveTangent=0.01, time=20, value=1}
     ]}}
  }
 }
 {id=Mixer_0, children [Modulator_0,WavePlayer_2], properties {inputVolume [0.6,1.2]}}
 {id=Mixer_1, children [Modulator_1,WavePlayer_0], properties {inputVolume [0.6,1.2]}}
 {id=Modulator_0, children [WavePlayer_1], properties {pitchMax=0.9, pitchMin=0.9}}
 {id=Modulator_1, children [WavePlayer_3], properties {pitchMax=0.9, pitchMin=0.9}}
 {id=QualityLevel_0, children [Enveloper_3,Enveloper_3,Enveloper_2,Enveloper_2]}
 {id=WavePlayer_0, properties {bLooping=true, soundWaveAssetPtr=/AmbientAudioContent/Waves/POI/Ambience_Bed_Int_Brick_Med_01_St}}
 {id=WavePlayer_1, properties {bLooping=true, soundWaveAssetPtr=/AmbientAudioContent/Waves/POI/Ambience_Bed_Int_Brick_Med_01_Quad}}
 {id=WavePlayer_2, properties {bLooping=true, soundWaveAssetPtr=/AmbientAudioContent/Waves/POI/Ambience_Bed_Int_Metal_Lg_01_Quad}}
 {id=WavePlayer_3, properties {bLooping=true, soundWaveAssetPtr=/AmbientAudioContent/Waves/POI/Ambience_Bed_Int_Metal_Lg_01_St}}
]
properties {
 attenuationSettings=&Fort_Ambient_Athena_Att
 maxDistance=2097152
 soundClassObject=&Fort_Ambient_Athena
 volumeMultiplier=1
 attenuationOverrides {bAttenuate=false}
}
