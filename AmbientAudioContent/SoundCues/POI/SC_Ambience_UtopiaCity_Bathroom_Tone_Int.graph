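# Sound cue for the looping indoor room tone of a Utopia City bathroom, routed to the Athena indoor ambient sound class.
# A quality-level root sends the first two levels to a mixer blending a small-room quad loop with a quieter Japanese bathroom interior loop.
# The last two levels use a mixer with only the bathroom loop, leaving the other input empty.
# It plays non-spatialized, with a concurrency override that prevents new instances at the limit.
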SoundCue /AmbientAudioContent/SoundCues/POI/SC_Ambience_UtopiaCity_Bathroom_Tone_Int
root=QualityLevel_0
nodes [
 {id=Mixer_0, children [WavePlayer_2,WavePlayer_0], properties {inputVolume [0.75,0.4]}}
 {id=Mixer_1, children [null,WavePlayer_1], properties {inputVolume [0.75,0.4]}}
 {id=QualityLevel_0, children [Mixer_0,Mixer_0,Mixer_1,Mixer_1]}
 {id=WavePlayer_0, properties {bLooping=true, soundWaveAssetPtr=/AmbientAudioContent/Waves/POI/Ambience_Japanese_Bathroom_Interior_Loop}}
 {id=WavePlayer_1, properties {bLooping=true, soundWaveAssetPtr=&Ambience_Japanese_Bathroom_Interior_Loop}}
 {id=WavePlayer_2, properties {bLooping=true, soundWaveAssetPtr=/AmbientAudioContent/Waves/POI/Ambience_UtopiaCity_Small_Room_Quad_Int}}
]
properties {
 bOverrideConcurrency=true
 maxDistance=2097152
 soundClassObject=&Fort_Ambient_Indoors_Athena
 volumeMultiplier=0.4
 attenuationOverrides {bAttenuate=false, bEnableReverbSend=false, bSpatialize=false}
 concurrencyOverrides {resolutionRule=PreventNew}
}
