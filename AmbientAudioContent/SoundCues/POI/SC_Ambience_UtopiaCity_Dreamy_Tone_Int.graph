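# Looping, non-spatialized indoor ambience cue for the Utopia City POI, layering a medium room tone with a dreamy pad.
# A quality switch selects quad waves for its first two inputs and stereo waves for the last two, through identically structured mixers.
# Looping volume envelopes of differing periods make the layers swell and dip out of step, while a modulator randomizes room-tone pitch.
# It uses the indoor ambient sound class and prevents new instances at the concurrency limit.
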SoundCue /AmbientAudioContent/SoundCues/POI/SC_Ambience_UtopiaCity_Dreamy_Tone_Int
root=QualityLevel_0
nodes [
 {
  id=Enveloper_0
  children [Modulator_2]
  properties {
   bLoop=true
   bLoopIndefinitely=true
   loopEnd=30
   pitchCurve {editorCurveData {keys [{time=0}]}}
   volumeCurve {editorCurveData {keys [<at 0 1>,<at 15 0.5>,<at 29.99 1>]}}
  }
 }
 {
  id=Enveloper_1
  children [Modulator_0]
  properties {
   bLoop=true
   bLoopIndefinitely=true
   loopEnd=30
   pitchCurve {editorCurveData {keys [{time=0}]}}
   volumeCurve {editorCurveData {keys [<at 0 1>,<at 15 0.5>,<at 29.99 1>]}}
  }
 }
 {
  id=Enveloper_2
  children [WavePlayer_1]
  properties {
   bLoop=true
   bLoopIndefinitely=true
   loopEnd=25
   pitchCurve {editorCurveData {keys [{time=0}]}}
   volumeCurve {editorCurveData {keys [<at 0 1>,<at 12.5 0.2>,<at 24.99 1>]}}
  }
 }
 {
  id=Enveloper_3
  children [WavePlayer_4]
  properties {
   bLoop=true
   bLoopIndefinitely=true
   loopEnd=25
   pitchCurve {editorCurveData {keys [{time=0}]}}
   volumeCurve {editorCurveData {keys [<at 0 1>,<at 12.5 0.2>,<at 24.99 1>]}}
  }
 }
 {id=Mixer_0, children [Enveloper_0,Enveloper_2], properties {inputVolume [0.6,0.9]}}
 {id=Mixer_1, children [Enveloper_1,Enveloper_3], properties {inputVolume [0.6,0.9]}}
 {id=Modulator_0, children [WavePlayer_3], properties {pitchMax=0.9, pitchMin=1.05}}
 {id=Modulator_2, children [WavePlayer_2], properties {pitchMax=0.9, pitchMin=1.05}}
 {id=QualityLevel_0, children [Mixer_0,Mixer_0,Mixer_1,Mixer_1]}
 {id=WavePlayer_1, properties {bLooping=true, soundWaveAssetPtr=/AmbientAudioContent/Waves/POI/Ambience_UtopiaCity_Dreamy_Pad_Quad_Loop}}
 {id=WavePlayer_2, properties {bLooping=true, soundWaveAssetPtr=/AmbientAudioContent/Waves/POI/Ambience_UtopiaCity_Medium_Room_Quad_Int}}
 {id=WavePlayer_3, properties {bLooping=true, soundWaveAssetPtr=/AmbientAudioContent/Waves/POI/Ambience_UtopiaCity_Medium_Room_Stereo_Int}}
 {id=WavePlayer_4, properties {bLooping=true, soundWaveAssetPtr=/AmbientAudioContent/Waves/POI/Ambience_UtopiaCity_Dreamy_Pad_Stereo_Loop}}
]
properties {
 bOverrideConcurrency=true
 maxDistance=2097152
 soundClassObject=&Fort_Ambient_Indoors_Athena
 volumeMultiplier=0.6
 attenuationOverrides {bAttenuate=false, bEnableReverbSend=false, bSpatialize=false}
 concurrencyOverrides {resolutionRule=PreventNew}
}
