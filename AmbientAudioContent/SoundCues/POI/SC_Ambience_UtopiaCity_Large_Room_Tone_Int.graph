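# Sound cue for the looping interior room-tone ambience of a large Utopia City POI space, routed to the Athena indoor ambient sound class.
# A quality-level node selects a looping quad wave for its first two levels and a looping stereo wave for the last two, each through a slight pitch-randomizing modulator.
# Overrides disable attenuation, spatialization and reverb send, and a concurrency override prevents new instances once the limit is reached.
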SoundCue /AmbientAudioContent/SoundCues/POI/SC_Ambience_UtopiaCity_Large_Room_Tone_Int
root=QualityLevel_0
nodes [
 {id=Modulator_0, children [WavePlayer_2], properties {pitchMax=0.9, pitchMin=1.05}}
 {id=Modulator_1, children [WavePlayer_3], properties {pitchMax=0.9, pitchMin=1.05}}
 {id=QualityLevel_0, children [Modulator_0,Modulator_0,Modulator_1,Modulator_1]}
 {id=WavePlayer_2, properties {bLooping=true, soundWaveAssetPtr=/AmbientAudioContent/Waves/POI/Ambience_UtopiaCity_Large_Room_Quad_Int}}
 {id=WavePlayer_3, properties {bLooping=true, soundWaveAssetPtr=/AmbientAudioContent/Waves/POI/Ambience_UtopiaCity_Large_Room_Stereo_Int}}
]
properties {
 bOverrideConcurrency=true
 maxDistance=2097152
 soundClassObject=&Fort_Ambient_Indoors_Athena
 volumeMultiplier=0.6
 attenuationOverrides {bAttenuate=false, bEnableReverbSend=false, bSpatialize=false}
 concurrencyOverrides {resolutionRule=PreventNew}
}
