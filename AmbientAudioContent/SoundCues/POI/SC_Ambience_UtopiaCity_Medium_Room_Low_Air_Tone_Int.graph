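# Looping, non-spatialized indoor room-tone cue for Utopia City medium rooms, in the Athena indoor ambient sound class.
# A quality switch feeds the first two levels from quad waves and the last two from stereo versions.
# Each branch mixes looping Utopia City and The Raft room beds through pitch modulators, with The Raft pitched down.
# A looping 30-second envelope slowly dips and restores the mix volume.
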SoundCue /AmbientAudioContent/SoundCues/POI/SC_Ambience_UtopiaCity_Medium_Room_Low_Air_Tone_Int
root=QualityLevel_1
nodes [
 {
  id=Enveloper_0
  children [Mixer_2]
  properties {
   bLoop=true
   bLoopIndefinitely=true
   loopEnd=30
   pitchCurve {editorCurveData {keys [{time=0}]}}
   volumeCurve {editorCurveData {keys [<at 0 1 RCIM_Cubic>,<at 15 0.5 RCIM_Cubic>,<at 29.99 1>]}}
  }
 }
 {
  id=Enveloper_1
  children [Mixer_3]
  properties {
   bLoop=true
   bLoopIndefinitely=true
   loopEnd=30
   pitchCurve {editorCurveData {keys [{time=0}]}}
   volumeCurve {editorCurveData {keys [<at 0 1 RCIM_Cubic>,<at 15 0.5 RCIM_Cubic>,<at 29.99 1>]}}
  }
 }
 {id=Mixer_2, children [Modulator_1,Modulator_0], properties {inputVolume [0.6,0.5]}}
 {id=Mixer_3, children [Modulator_3,Modulator_2], properties {inputVolume [0.6,0.5]}}
 {id=Modulator_0, children [WavePlayer_8], properties {pitchMax=0.85, pitchMin=0.8}}
 {id=Modulator_1, children [WavePlayer_6], properties {pitchMax=0.9, pitchMin=1.05}}
 {id=Modulator_2, children [WavePlayer_9], properties {pitchMax=0.85, pitchMin=0.8}}
 {id=Modulator_3, children [WavePlayer_7], properties {pitchMax=0.9, pitchMin=1.05}}
 {id=QualityLevel_1, children [Enveloper_0,Enveloper_0,Enveloper_1,Enveloper_1]}
 {id=WavePlayer_6, properties {bLooping=true, soundWaveAssetPtr=/AmbientAudioContent/Waves/POI/Ambience_UtopiaCity_Medium_Room_Quad_Int}}
 {id=WavePlayer_7, properties {bLooping=true, soundWaveAssetPtr=/AmbientAudioContent/Waves/POI/Ambience_UtopiaCity_Medium_Room_Stereo_Int}}
 {id=WavePlayer_8, properties {bLooping=true, soundWaveAssetPtr=/AmbientAudioContent/Waves/POI/Ambience_TheRaft_Med_Room_01_Int_Quad}}
 {id=WavePlayer_9, properties {bLooping=true, soundWaveAssetPtr=/AmbientAudioContent/Waves/POI/Ambience_TheRaft_Med_Room_01_Int_Stereo}}
]
properties {
 bOverrideConcurrency=true
 maxDistance=2097152
 soundClassObject=&Fort_Ambient_Indoors_Athena
 volumeMultiplier=0.55
 attenuationOverrides {bAttenuate=false, bEnableReverbSend=false, bSpatialize=false}
 concurrencyOverrides {resolutionRule=PreventNew}
}
