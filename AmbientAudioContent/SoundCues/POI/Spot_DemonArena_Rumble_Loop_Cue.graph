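# Looping ambient cue for a Demon Arena POI rumble, built from sky island rumble, rock crumble and glowing rock sounds.
# Layers with their own attenuation and sound class combine sub, mid and glow loops with randomly delayed rumble and debris one-shots.
# A quality switch selects either this full mix or a simpler glow-plus-sub branch.
# Nested looping envelopers slowly sway pitch and volume so the bed keeps evolving.
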SoundCue /AmbientAudioContent/SoundCues/POI/Spot_DemonArena_Rumble_Loop_Cue
root=Enveloper_3
nodes [
 {
  id=Attenuation_0
  children [SoundClass_2]
  properties {attenuationSettings=/AmbientAudioBase/Attenuations/Spot_SkyIsland_Underside_Med_Narrow_Att}
 }
 {
  id=Attenuation_1
  children [SoundClass_0]
  properties {attenuationSettings=/AmbientAudioBase/Attenuations/Spot_SkyIsland_Underside_Medium_Sphere_Att}
 }
 {
  id=Attenuation_2
  children [SoundClass_3]
  properties {attenuationSettings=/AmbientAudioBase/Attenuations/Spot_SkyIsland_Underside_Detail_Narrow_Att}
 }
 {
  id=Attenuation_3
  children [SoundClass_0]
  properties {attenuationSettings=/AmbientAudioBase/Attenuations/Spot_SkyIsland_Underside_Far_Narrow_Att}
 }
 {
  id=Attenuation_4
  children [SoundClass_4]
  properties {attenuationSettings=&Spot_SkyIsland_Underside_Far_Narrow_Att}
 }
 {
  id=Attenuation_5
  children [SoundClass_5]
  properties {attenuationSettings=&Spot_SkyIsland_Underside_Med_Narrow_Att}
 }
 {
  id=Attenuation_6
  children [SoundClass_2]
  properties {attenuationSettings=&Spot_SkyIsland_Underside_Medium_Sphere_Att}
 }
 {id=Delay_0, children [Random_8], properties {delayMax=12, delayMin=3}}
 {id=Delay_2, children [Random_3], properties {delayMax=8, delayMin=3}}
 {
  id=Enveloper_0
  children [QualityLevel_1]
  properties {
   bLoop=true
   bLoopIndefinitely=true
   loopEnd=30
   pitchCurve {editorCurveData {keys [{arriveTangent=-0.0013, interpMode=RCIM_Cubic, time=0}, <at 15 0.95 RCIM_Cubic>, {interpMode=RCIM_Cubic, leaveTangent=0.0013, time=30}]}}
   volumeCurve {editorCurveData {keys [<at 0 1>]}}
  }
 }
 {
  id=Enveloper_1
  children [Modulator_0]
  properties {
   bLoop=true
   bLoopIndefinitely=true
   loopEnd=30
   pitchCurve {editorCurveData {keys [{time=0}]}}
   volumeCurve {editorCurveData {keys [
      {arriveTangent=-0.01, interpMode=RCIM_Cubic, time=0, value=1}
      <at 15 0.7 RCIM_Cubic>
      {interpMode=RCIM_Cubic, leaveTangent=0.01, time=30, value=1}
     ]}}
  }
 }
 {
  id=Enveloper_3
  children [Enveloper_0]
  properties {
   bLoop=true
   bLoopIndefinitely=true
   loopEnd=20
   pitchCurve {editorCurveData {keys [{time=0}]}}
   volumeCurve {editorCurveData {keys [
      {arriveTangent=-0.01, interpMode=RCIM_Cubic, time=0, value=1}
      <at 10 0.5 RCIM_Cubic>
      {interpMode=RCIM_Cubic, leaveTangent=0.01, time=20, value=1}
     ]}}
  }
 }
 {
  id=Enveloper_4
  children [Enveloper_1]
  properties {
   bLoop=true
   bLoopIndefinitely=true
   loopEnd=60
   pitchCurve {editorCurveData {keys [
      {arriveTangent=-0.0013, interpMode=RCIM_Cubic, time=0}
      {arriveTangent=-0.0004, interpMode=RCIM_Cubic, leaveTangent=-0.0004, tangentMode=RCTM_User, time=15}
      <at 30 0.95 RCIM_Cubic>
      {time=45}
      {interpMode=RCIM_Cubic, leaveTangent=0.0013, time=60}
     ]}}
   volumeCurve {editorCurveData {keys [<at 0 1>]}}
  }
 }
 {
  id=Enveloper_5
  children [Modulator_2]
  properties {
   bLoop=true
   bLoopIndefinitely=true
   loopEnd=30
   pitchCurve {editorCurveData {keys [{time=0}]}}
   volumeCurve {editorCurveData {keys [
      {arriveTangent=-0.01, interpMode=RCIM_Cubic, time=0, value=1}
      <at 15 0.5 RCIM_Cubic>
      {interpMode=RCIM_Cubic, leaveTangent=0.01, time=30, value=1}
     ]}}
  }
 }
 {id=Looping_0, children [Delay_0]}
 {id=Looping_2, children [Delay_2]}
 {id=Mixer_0, children [Mixer_2, Attenuation_5, Mixer_7, Attenuation_2], properties {inputVolume [1,1.6,1.5,2]}}
 {id=Mixer_1, children [Looping_0,WavePlayer_8], properties {inputVolume [0.5,0.8]}}
 {id=Mixer_10, children [Random_2], properties {inputVolume [1.25]}}
 {id=Mixer_2, children [Attenuation_1, Attenuation_3], properties {inputVolume [1,1]}}
 {id=Mixer_3, children [WavePlayer_20, WavePlayer_21], properties {inputVolume [1,1]}}
 {id=Mixer_4, children [Looping_2], properties {inputVolume [0.15]}}
 {id=Mixer_5, children [WavePlayer_9], properties {inputVolume [1.5]}}
 {id=Mixer_6, children [Random_13], properties {inputVolume [1]}}
 {id=Mixer_7, children [Attenuation_6, Attenuation_0], properties {inputVolume [1,1]}}
 {id=Mixer_8, children [Random_14], properties {inputVolume [2]}}
 {id=Mixer_9, children [Random_12], properties {inputVolume [1.25]}}
 {id=Modulator_0, children [WavePlayer_0]}
 {id=Modulator_2, children [Mixer_3]}
 {id=QualityLevel_1, children [Mixer_0, Mixer_0, Attenuation_4, Attenuation_4]}
 {id=Random_12, children [WavePlayer_62, WavePlayer_63, WavePlayer_64, WavePlayer_65], properties {weights [1,1,1,1]}}
 {id=Random_13, children [WavePlayer_66, WavePlayer_67, WavePlayer_68, WavePlayer_69, WavePlayer_70], properties {weights [1,1,1,1,1]}}
 {id=Random_14, children [WavePlayer_71, WavePlayer_72, WavePlayer_73, WavePlayer_74], properties {weights [1,1,1,1]}}
 {
  id=Random_2
  children [WavePlayer_3, WavePlayer_4, WavePlayer_11, WavePlayer_15, WavePlayer_16, WavePlayer_17, WavePlayer_18, WavePlayer_19, WavePlayer_1]
  properties {weights [1,1,1,1,1,1,1,1,1]}
 }
 {id=Random_3, children [Mixer_9,Mixer_6,Mixer_8,Mixer_10], properties {weights [1,1,1,1]}}
 {id=Random_8, children [WavePlayer_27, WavePlayer_28, WavePlayer_42, WavePlayer_43], properties {weights [1,1,1,1]}}
 {id=SoundClass_0, children [Mixer_1], properties {soundClassOverride=&Fort_Gadgets}}
 {id=SoundClass_2, children [Mixer_5], properties {soundClassOverride=&Fort_Ambient_Outdoors_Athena}}
 {id=SoundClass_3, children [Mixer_4], properties {soundClassOverride=&Fort_Ambient_Outdoors_Athena}}
 {id=SoundClass_4, children [Enveloper_5], properties {soundClassOverride=&Fort_Gadgets}}
 {id=SoundClass_5, children [Enveloper_4], properties {soundClassOverride=&Fort_Ambient_Outdoors_Athena}}
 {id=WavePlayer_0, properties {bLooping=true, soundWaveAssetPtr=/AmbientAudioContent/Waves/SpottedLoops/Spot_Fornitanium_Glowing_Rock_Loop}}
 {id=WavePlayer_1, properties {soundWaveAssetPtr=/AmbientAudioContent/Waves/OneShots/Spot_Cave_Rock_Debris_Falling_09}}
 {id=WavePlayer_11, properties {soundWaveAssetPtr=/AmbientAudioContent/Waves/OneShots/Spot_Cave_Rock_Debris_Falling_03}}
 {id=WavePlayer_15, properties {soundWaveAssetPtr=/AmbientAudioContent/Waves/OneShots/Spot_Cave_Rock_Debris_Falling_04}}
 {id=WavePlayer_16, properties {soundWaveAssetPtr=/AmbientAudioContent/Waves/OneShots/Spot_Cave_Rock_Debris_Falling_05}}
 {id=WavePlayer_17, properties {soundWaveAssetPtr=/AmbientAudioContent/Waves/OneShots/Spot_Cave_Rock_Debris_Falling_06}}
 {id=WavePlayer_18, properties {soundWaveAssetPtr=/AmbientAudioContent/Waves/OneShots/Spot_Cave_Rock_Debris_Falling_07}}
 {id=WavePlayer_19, properties {soundWaveAssetPtr=/AmbientAudioContent/Waves/OneShots/Spot_Cave_Rock_Debris_Falling_08}}
 {id=WavePlayer_20, properties {bLooping=true, soundWaveAssetPtr=&Spot_Fornitanium_Glowing_Rock_Loop}}
 {
  id=WavePlayer_21
  properties {bLooping=true, soundWaveAssetPtr=/AmbientAudioContent/Waves/SpottedLoops/Spot_SkyIsland_Rumble_Sub_Presence_Loop}
 }
 {id=WavePlayer_27, properties {soundWaveAssetPtr=/AmbientAudioContent/Waves/OneShots/Spot_SkyIsland_Rumble_Oneshot_01}}
 {id=WavePlayer_28, properties {soundWaveAssetPtr=/AmbientAudioContent/Waves/OneShots/Spot_SkyIsland_Rumble_Oneshot_02}}
 {id=WavePlayer_3, properties {soundWaveAssetPtr=/AmbientAudioContent/Waves/OneShots/Spot_Cave_Rock_Debris_Falling_01}}
 {id=WavePlayer_4, properties {soundWaveAssetPtr=/AmbientAudioContent/Waves/OneShots/Spot_Cave_Rock_Debris_Falling_02}}
 {id=WavePlayer_42, properties {soundWaveAssetPtr=/AmbientAudioContent/Waves/OneShots/Spot_SkyIsland_Rumble_Oneshot_04}}
 {id=WavePlayer_43, properties {soundWaveAssetPtr=/AmbientAudioContent/Waves/OneShots/Spot_SkyIsland_Rumble_Oneshot_05}}
 {id=WavePlayer_62, properties {soundWaveAssetPtr=/AmbientAudioContent/Waves/OneShots/Spot_Mountain_Rock_Slide_01}}
 {id=WavePlayer_63, properties {soundWaveAssetPtr=/AmbientAudioContent/Waves/OneShots/Spot_Mountain_Rock_Slide_02}}
 {id=WavePlayer_64, properties {soundWaveAssetPtr=/AmbientAudioContent/Waves/OneShots/Spot_Mountain_Rock_Slide_03}}
 {id=WavePlayer_65, properties {soundWaveAssetPtr=/AmbientAudioContent/Waves/OneShots/Spot_Mountain_Rock_Slide_04}}
 {id=WavePlayer_66, properties {soundWaveAssetPtr=/AmbientAudioContent/Waves/OneShots/Spot_Cliff_Dirt_01}}
 {id=WavePlayer_67, properties {soundWaveAssetPtr=/AmbientAudioContent/Waves/OneShots/Spot_Cliff_Dirt_02}}
 {id=WavePlayer_68, properties {soundWaveAssetPtr=/AmbientAudioContent/Waves/OneShots/Spot_Cliff_Dirt_03}}
 {id=WavePlayer_69, properties {soundWaveAssetPtr=/AmbientAudioContent/Waves/OneShots/Spot_Cliff_Dirt_04}}
 {id=WavePlayer_70, properties {soundWaveAssetPtr=/AmbientAudioContent/Waves/OneShots/Spot_Cliff_Dirt_05}}
 {id=WavePlayer_71, properties {soundWaveAssetPtr=/AmbientAudioContent/Waves/OneShots/Spot_Cliff_Sand_01}}
 {id=WavePlayer_72, properties {soundWaveAssetPtr=/AmbientAudioContent/Waves/OneShots/Spot_Cliff_Sand_02}}
 {id=WavePlayer_73, properties {soundWaveAssetPtr=/AmbientAudioContent/Waves/OneShots/Spot_Cliff_Sand_03}}
 {id=WavePlayer_74, properties {soundWaveAssetPtr=/AmbientAudioContent/Waves/OneShots/Spot_Cliff_Sand_04}}
 {
  id=WavePlayer_8
  properties {bLooping=true, soundWaveAssetPtr=&Spot_SkyIsland_Rumble_Sub_Presence_Loop}
 }
 {id=WavePlayer_9, properties {bLooping=true, soundWaveAssetPtr=/AmbientAudioContent/Waves/SpottedLoops/Spot_SkyIsland_Rock_Crumble_Mids_Loop}}
]
properties {
 bOverrideConcurrency=true
 maxDistance=20000
 soundClassObject=/Game/Sounds/FortSoundClassesAndModes/Fort_Ambient_Global
 volumeMultiplier=0.251
 attenuationOverrides {
  bEnableFocusInterpolation=true
  bEnableListenerFocus=true
  falloffDistance=500
  focusAttackInterpSpeed=0.01
  focusReleaseInterpSpeed=0.01
  focusVolumeAttenuation=1.2
  nonFocusVolumeAttenuation=0.75
  attenuationShapeExtents=<250 500>
 }
 concurrencyOverrides {maxCount=2, volumeScaleMode=Distance}
 preEffectBusSends [{sendLevel=0}]
}
