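# Looping outdoor ambient cue for a monolith spire point of interest, playing a rumble with slow, endlessly looping volume and pitch envelopes.
# A quality switch gives the first two levels randomly delayed rumble one-shots, a sub-presence loop and a pitched-down, enveloped glowing-rock loop.
# The other two levels get a simpler enveloped mix of both loops.
# Overrides set capsule HRTF attenuation with listener focus and limit concurrency to two.
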SoundCue /AmbientAudioContent/SoundCues/POI/Spot_Monolith_Spire_Rumble_Loop_Cue
root=Enveloper_3
nodes [
 {id=Delay_0, children [Random_8], properties {delayMax=12, delayMin=3}}
 {
  id=Enveloper_0
  children [QualityLevel_1]
  properties {
   bLoop=true
   bLoopIndefinitely=true
   loopEnd=30
   pitchCurve {editorCurveData {keys [{arriveTangent=-0.0013, interpMode=RCIM_Cubic, time=0}, <at 15 0.95 RCIM_Cubic>, {interpMode=RCIM_Cubic, leaveTangent=0.0013, time=30}]}}
   volumeCurve {editorCurveData {keys [<at 0 1>]}}
  }
 }
 {
  id=Enveloper_1
  children [Modulator_0]
  properties {
   bLoop=true
   bLoopIndefinitely=true
   loopEnd=30
   pitchCurve {editorCurveData {keys [{time=0}]}}
   volumeCurve {editorCurveData {keys [
      {arriveTangent=-0.01, interpMode=RCIM_Cubic, time=0, value=1}
      <at 15 0.7 RCIM_Cubic>
      {interpMode=RCIM_Cubic, leaveTangent=0.01, time=30, value=1}
     ]}}
  }
 }
 {
  id=Enveloper_3
  children [Enveloper_0]
  properties {
   bLoop=true
   bLoopIndefinitely=true
   loopEnd=20
   pitchCurve {editorCurveData {keys [{time=0}]}}
   volumeCurve {editorCurveData {keys [
      {arriveTangent=-0.01, interpMode=RCIM_Cubic, time=0, value=1}
      <at 10 0.5 RCIM_Cubic>
      {interpMode=RCIM_Cubic, leaveTangent=0.01, time=20, value=1}
     ]}}
  }
 }
 {
  id=Enveloper_4
  children [Enveloper_1]
  properties {
   bLoop=true
   bLoopIndefinitely=true
   loopEnd=60
   pitchCurve {editorCurveData {keys [
      {arriveTangent=-0.0013, interpMode=RCIM_Cubic, time=0}
      {arriveTangent=-0.0004, interpMode=RCIM_Cubic, leaveTangent=-0.0004, tangentMode=RCTM_User, time=15}
      <at 30 0.95 RCIM_Cubic>
      {time=45}
      {interpMode=RCIM_Cubic, leaveTangent=0.0013, time=60}
     ]}}
   volumeCurve {editorCurveData {keys [<at 0 1>]}}
  }
 }
 {
  id=Enveloper_5
  children [Modulator_2]
  properties {
   bLoop=true
   bLoopIndefinitely=true
   loopEnd=30
   pitchCurve {editorCurveData {keys [{time=0}]}}
   volumeCurve {editorCurveData {keys [
      {arriveTangent=-0.01, interpMode=RCIM_Cubic, time=0, value=1}
      <at 15 0.5 RCIM_Cubic>
      {interpMode=RCIM_Cubic, leaveTangent=0.01, time=30, value=1}
     ]}}
  }
 }
 {id=Looping_0, children [Delay_0]}
 {id=Mixer_0, children [Mixer_1,Enveloper_4], properties {inputVolume [1,1]}}
 {id=Mixer_1, children [Looping_0,WavePlayer_8,null], properties {inputVolume [0.4,1,0.8]}}
 {id=Mixer_3, children [WavePlayer_20, WavePlayer_21], properties {inputVolume [2,1]}}
 {id=Modulator_0, children [WavePlayer_0], properties {pitchMax=0.6, pitchMin=0.6}}
 {id=Modulator_2, children [Mixer_3]}
 {id=QualityLevel_1, children [Mixer_0,Mixer_0,Enveloper_5,Enveloper_5]}
 {id=Random_8, children [WavePlayer_27, WavePlayer_28, WavePlayer_42, WavePlayer_43], properties {weights [1,1,1,1]}}
 {id=WavePlayer_0, properties {bLooping=true, soundWaveAssetPtr=/AmbientAudioContent/Waves/SpottedLoops/Spot_Fornitanium_Glowing_Rock_Loop}}
 {id=WavePlayer_20, properties {bLooping=true, soundWaveAssetPtr=&Spot_Fornitanium_Glowing_Rock_Loop}}
 {
  id=WavePlayer_21
  properties {bLooping=true, soundWaveAssetPtr=/AmbientAudioContent/Waves/SpottedLoops/Spot_SkyIsland_Rumble_Sub_Presence_Loop}
 }
 {id=WavePlayer_27, properties {soundWaveAssetPtr=/AmbientAudioContent/Waves/OneShots/Spot_SkyIsland_Rumble_Oneshot_01}}
 {id=WavePlayer_28, properties {soundWaveAssetPtr=/AmbientAudioContent/Waves/OneShots/Spot_SkyIsland_Rumble_Oneshot_02}}
 {id=WavePlayer_42, properties {soundWaveAssetPtr=/AmbientAudioContent/Waves/OneShots/Spot_SkyIsland_Rumble_Oneshot_04}}
 {id=WavePlayer_43, properties {soundWaveAssetPtr=/AmbientAudioContent/Waves/OneShots/Spot_SkyIsland_Rumble_Oneshot_05}}
 {
  id=WavePlayer_8
  properties {bLooping=true, soundWaveAssetPtr=&Spot_SkyIsland_Rumble_Sub_Presence_Loop}
 }
]
properties {
 bOverrideAttenuation=true
 bOverrideConcurrency=true
 maxDistance=7278.3135
 soundClassObject=&Fort_Ambient_Outdoors_Athena
 volumeMultiplier=1
 attenuationOverrides {
  attenuationShape=Capsule
  bEnableListenerFocus=true
  falloffDistance=4500
  focusAttackInterpSpeed=0.01
  focusAzimuth=15
  focusReleaseInterpSpeed=0.01
  focusVolumeAttenuation=1.2
  nonFocusAzimuth=45
  nonFocusVolumeAttenuation=0.5
  spatializationAlgorithm=SPATIALIZATION_HRTF
  stereoSpread=10
  attenuationShapeExtents {x=2778.3135}
 }
 concurrencyOverrides {maxCount=2, volumeScaleMode=Distance}
 preEffectBusSends [{sendLevel=0}]
}
