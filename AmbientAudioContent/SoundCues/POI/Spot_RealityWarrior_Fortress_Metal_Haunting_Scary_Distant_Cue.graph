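# Sound cue for a distant, haunting metal one-shot ambience, named for the Reality Warrior Fortress point of interest.
# A Random node picks among five equally weighted stereo blimp metal waves, feeding a Modulator for pitch variation, then an Enveloper with a pitch curve.
# The cue halves overall pitch, repurposing the blimp recordings at a lower register.
# It uses the indoor ambient one-shot sound class and less-positional attenuation, with bus and submix sends disabled.
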SoundCue /AmbientAudioContent/SoundCues/POI/Spot_RealityWarrior_Fortress_Metal_Haunting_Scary_Distant_Cue
root=Enveloper_0
nodes [
 {
  id=Enveloper_0
  children [Modulator_0]
  properties {pitchCurve {editorCurveData {keys [{time=0}, {time=3}, <at 10 0.9>]}}, volumeCurve {editorCurveData {keys [<at 0 1>]}}}
 }
 {id=Modulator_0, children [Random_0], properties {pitchMin=0.9}}
 {id=Random_0, children [WavePlayer_5, WavePlayer_6, WavePlayer_7, WavePlayer_13, WavePlayer_14], properties {weights [1,1,1,1,1]}}
 {id=WavePlayer_13, properties {soundWaveAssetPtr=/AmbientAudioContent/Waves/OneShots/Spot_Blimp_Metal_Haunting_Scary_Distant_04_Stereo}}
 {id=WavePlayer_14, properties {soundWaveAssetPtr=/AmbientAudioContent/Waves/OneShots/Spot_Blimp_Metal_Haunting_Scary_Distant_05_Stereo}}
 {id=WavePlayer_5, properties {soundWaveAssetPtr=/AmbientAudioContent/Waves/OneShots/Spot_Blimp_Metal_Haunting_Scary_Distant_01_Stereo}}
 {id=WavePlayer_6, properties {soundWaveAssetPtr=/AmbientAudioContent/Waves/OneShots/Spot_Blimp_Metal_Haunting_Scary_Distant_02_Stereo}}
 {id=WavePlayer_7, properties {soundWaveAssetPtr=/AmbientAudioContent/Waves/OneShots/Spot_Blimp_Metal_Haunting_Scary_Distant_03_Stereo}}
]
properties {
 attenuationSettings=/Game/Sounds/Attenuation/Fort_TileAmbient_OneShot_LessPositional_Log_Att
 bEnableBusSends=false
 bEnableSubmixSends=false
 maxDistance=6200
 pitchMultiplier=0.5
 soundClassObject=/Game/Sounds/FortSoundClassesAndModes/Fort_Ambient_OneShots_Indoors
 volumeMultiplier=1.253
 attenuationOverrides {falloffDistance=6000, stereoSpread=10000, attenuationShapeExtents {x=2000}}
}
