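# Ambient one-shot SoundCue that plays an ominous wind swell for the Reality Warrior Fortress point of interest.
# An equal-weight Random node picks one of seven cave wind swell waves and feeds an Enveloper, which keys pitch over time and holds volume flat.
# The cue also raises pitch and lowers volume, using tile-ambient attenuation and the indoor ambient one-shot sound class, with bus and submix sends disabled.
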SoundCue /AmbientAudioContent/SoundCues/POI/Spot_RealityWarrior_Fortress_Ominous_Wind_Swells_Cue
root=Enveloper_0
nodes [
 {
  id=Enveloper_0
  children [Random_1]
  properties {pitchCurve {editorCurveData {keys [{time=0}, {time=4}, <at 8 1.2>]}}, volumeCurve {editorCurveData {keys [<at 0 1>]}}}
 }
 {
  id=Random_1
  children [WavePlayer_0,WavePlayer_1,WavePlayer_2,WavePlayer_3,WavePlayer_4,WavePlayer_5,WavePlayer_6]
  properties {weights [1,1,1,1,1,1,1]}
 }
 {id=WavePlayer_0, properties {soundWaveAssetPtr=/AmbientAudioContent/Waves/OneShots/Spot_Cave_Ominous_Wind_Swells_01}}
 {id=WavePlayer_1, properties {soundWaveAssetPtr=/AmbientAudioContent/Waves/OneShots/Spot_Cave_Ominous_Wind_Swells_02}}
 {id=WavePlayer_2, properties {soundWaveAssetPtr=/AmbientAudioContent/Waves/OneShots/Spot_Cave_Ominous_Wind_Swells_03}}
 {id=WavePlayer_3, properties {soundWaveAssetPtr=/AmbientAudioContent/Waves/OneShots/Spot_Cave_Ominous_Wind_Swells_04}}
 {id=WavePlayer_4, properties {soundWaveAssetPtr=/AmbientAudioContent/Waves/OneShots/Spot_Cave_Ominous_Wind_Swells_05}}
 {id=WavePlayer_5, properties {soundWaveAssetPtr=/AmbientAudioContent/Waves/OneShots/Spot_Cave_Ominous_Wind_Swells_06}}
 {id=WavePlayer_6, properties {soundWaveAssetPtr=/AmbientAudioContent/Waves/OneShots/Spot_Cave_Ominous_Wind_Swells_07}}
]
properties {
 attenuationSettings=/Game/Sounds/Attenuation/Fort_TileAmbient_OneShot_Att
 bEnableBusSends=false
 bEnableSubmixSends=false
 maxDistance=8000
 pitchMultiplier=1.2
 soundClassObject=/Game/Sounds/FortSoundClassesAndModes/Fort_Ambient_OneShots_Indoors
 volumeMultiplier=0.251
}
