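# Sound cue for a distant town clock chime, an outdoor ambient one-shot.
# A random node picks one of four chime waves, two of them pitched down and quietened by modulators.
# The result gets slight volume variation and an envelope that fades out after a short hold, shared by all quality levels.
# Attenuation overrides add distance low-pass filtering and a distant reverb send, while concurrency blocks overlapping plays.
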SoundCue /AmbientAudioContent/SoundCues/POI/Spot_Town_Clock_Chime_Distant_Cue
root=QualityLevel_0
nodes [
 {
  id=Enveloper_2
  children [Modulator_2]
  properties {
   pitchCurve {editorCurveData {keys [{time=0}]}}
   volumeCurve {editorCurveData {keys [
      <at 0 1 RCIM_Cubic>
      {arriveTangent=-0.0158, interpMode=RCIM_Cubic, leaveTangent=-0.0158, tangentMode=RCTM_User, time=2, value=1}
      {arriveTangent=0, interpMode=RCIM_Cubic, leaveTangent=0, tangentMode=RCTM_User, time=4, value=0}
     ]}}
  }
 }
 {id=Modulator_2, children [Random_1], properties {volumeMin=0.8}}
 {id=Modulator_6, children [WavePlayer_0], properties {pitchMax=0.6, pitchMin=0.5, volumeMax=0.5, volumeMin=0.3}}
 {id=Modulator_7, children [WavePlayer_1], properties {pitchMax=0.6, pitchMin=0.5, volumeMax=0.5, volumeMin=0.3}}
 {id=QualityLevel_0, children [Enveloper_2,Enveloper_2,Enveloper_2,Enveloper_2]}
 {id=Random_1, children [Modulator_6,Modulator_7,WavePlayer_2,WavePlayer_3], properties {weights [1,1,1,1]}}
 {id=WavePlayer_0, properties {soundWaveAssetPtr=/AmbientAudioContent/Waves/OneShots/Spot_Town_Clock_Chime_01}}
 {id=WavePlayer_1, properties {soundWaveAssetPtr=/AmbientAudioContent/Waves/OneShots/Spot_Town_Clock_Chime_02}}
 {id=WavePlayer_2, properties {soundWaveAssetPtr=/AmbientAudioContent/Waves/OneShots/Spot_Town_Clock_Chime_03}}
 {id=WavePlayer_3, properties {soundWaveAssetPtr=/AmbientAudioContent/Waves/OneShots/Spot_Town_Clock_Chime_04}}
]
properties {
 bEnableBusSends=false
 bOverrideAttenuation=true
 bOverrideConcurrency=true
 maxDistance=8000
 soundClassObject=&Fort_Ambient_OneShots_Outdoors
 volumeMultiplier=1.504
 attenuationOverrides {
  bAttenuateWithLPF=true
  bEnableSubmixSends=true
  falloffDistance=8000
  lPFFrequencyAtMax=1000
  lPFRadiusMax=8000
  lPFRadiusMin=5000
  attenuationShapeExtents {x=0}
  submixSendSettings [{
    bEnableHPFCutoff=false
    bEnableLPFCutoff=false
    disableManualSendClamp=false
    hPFCutoff=20
    lPFCutoff=20000
    maxSendDistance=7500
    maxSendLevel=0.4
    minSendDistance=0
    minSendLevel=0
    sendLevel=0.2
    sendLevelControlMethod=Linear
    soundSubmix=&Fortnite_Reverb_Distant
   }]
 }
 concurrencyOverrides {maxCount=1, resolutionRule=PreventNew}
 preEffectBusSends [{}]
}
