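# Sound cue for one-shot flying trash ambience at town points of interest, attenuated and capped by a spot concurrency asset.
# A quality-level switch sends its first two levels to a random pick of eight waves and its last two to a random pick of four of those same waves.
# Each branch passes through a modulator that varies pitch and volume, so the reduced pool keeps some variation.
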SoundCue /AmbientAudioContent/SoundCues/POI/Spot_Town_Flying_Trash_Cue
root=QualityLevel_0
nodes [
 {id=Modulator_1, children [Random_0], properties {pitchMax=1.1, volumeMax=0.7, volumeMin=0.6}}
 {id=Modulator_2, children [Random_3], properties {pitchMax=1.1, volumeMax=0.7, volumeMin=0.6}}
 {id=QualityLevel_0, children [Modulator_1,Modulator_1,Modulator_2,Modulator_2]}
 {
  id=Random_0
  children [WavePlayer_0,WavePlayer_1,WavePlayer_2,WavePlayer_3,WavePlayer_4,WavePlayer_5,WavePlayer_6,WavePlayer_7]
  properties {weights [1,1,1,1,1,1,1,1]}
 }
 {id=Random_3, children [WavePlayer_16, WavePlayer_17, WavePlayer_18, WavePlayer_19], properties {weights [1,1,1,1]}}
 {id=WavePlayer_0, properties {soundWaveAssetPtr=/AmbientAudioContent/Waves/OneShots/Spot_Cities_Flying_Trash_01}}
 {id=WavePlayer_1, properties {soundWaveAssetPtr=/AmbientAudioContent/Waves/OneShots/Spot_Cities_Flying_Trash_02}}
 {id=WavePlayer_16, properties {soundWaveAssetPtr=/AmbientAudioContent/Waves/OneShots/Spot_Cities_Flying_Trash_06}}
 {id=WavePlayer_17, properties {soundWaveAssetPtr=/AmbientAudioContent/Waves/OneShots/Spot_Cities_Flying_Trash_03}}
 {id=WavePlayer_18, properties {soundWaveAssetPtr=&Spot_Cities_Flying_Trash_02}}
 {id=WavePlayer_19, properties {soundWaveAssetPtr=&Spot_Cities_Flying_Trash_01}}
 {id=WavePlayer_2, properties {soundWaveAssetPtr=&Spot_Cities_Flying_Trash_03}}
 {id=WavePlayer_3, properties {soundWaveAssetPtr=/AmbientAudioContent/Waves/OneShots/Spot_Cities_Flying_Trash_04}}
 {id=WavePlayer_4, properties {soundWaveAssetPtr=/AmbientAudioContent/Waves/OneShots/Spot_Cities_Flying_Trash_05}}
 {id=WavePlayer_5, properties {soundWaveAssetPtr=&Spot_Cities_Flying_Trash_06}}
 {id=WavePlayer_6, properties {soundWaveAssetPtr=/AmbientAudioContent/Waves/OneShots/Spot_Cities_Flying_Trash_07}}
 {id=WavePlayer_7, properties {soundWaveAssetPtr=/AmbientAudioContent/Waves/OneShots/Spot_Cities_Flying_Trash_08}}
]
properties {
 attenuationSettings=/Game/Sounds/Attenuation/Fort_TileAmbient_OneShot_Wider_Att
 bEnableBusSends=false
 maxDistance=8000
 soundClassObject=&Fort_Ambient_OneShots
 volumeMultiplier=0.351
 concurrencySet [/AmbientAudioBase/Concurrency/AmbientConc_2Max_Spot_Concurrency]
 preEffectBusSends [{}]
}
