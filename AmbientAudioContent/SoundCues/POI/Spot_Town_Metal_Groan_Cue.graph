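# Sound cue for a metal groan ambient one-shot, named as a town POI spot sound.
# A Random node picks among six equally weighted groan waves, and a Modulator varies pitch and volume on each play.
# The root Quality Level node feeds all four levels from that same Modulator, so no level gets a reduced version.
# It uses the distant creak one-shot attenuation and the ambient one-shots sound class, at reduced volume.
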SoundCue /AmbientAudioContent/SoundCues/POI/Spot_Town_Metal_Groan_Cue
root=QualityLevel_0
nodes [
 {id=Modulator_1, children [Random_2], properties {pitchMax=1.2, pitchMin=0.9, volumeMax=1.1, volumeMin=0.9}}
 {id=QualityLevel_0, children [Modulator_1,Modulator_1,Modulator_1,Modulator_1]}
 {
  id=Random_2
  children [WavePlayer_8, WavePlayer_9, WavePlayer_10, WavePlayer_11, WavePlayer_12, WavePlayer_13]
  properties {weights [1,1,1,1,1,1]}
 }
 {id=WavePlayer_10, properties {soundWaveAssetPtr=/AmbientAudioContent/Waves/OneShots/Spot_Town_Metal_Groan_03}}
 {id=WavePlayer_11, properties {soundWaveAssetPtr=/AmbientAudioContent/Waves/OneShots/Spot_Town_Metal_Groan_04}}
 {id=WavePlayer_12, properties {soundWaveAssetPtr=/AmbientAudioContent/Waves/OneShots/Spot_Town_Metal_Groan_05}}
 {id=WavePlayer_13, properties {soundWaveAssetPtr=/AmbientAudioContent/Waves/OneShots/Spot_Town_Metal_Groan_06}}
 {id=WavePlayer_8, properties {soundWaveAssetPtr=/AmbientAudioContent/Waves/OneShots/Spot_Town_Metal_Groan_01}}
 {id=WavePlayer_9, properties {soundWaveAssetPtr=/AmbientAudioContent/Waves/OneShots/Spot_Town_Metal_Groan_02}}
]
properties {
 attenuationSettings=/Game/Sounds/Attenuation/Fort_TileAmbient_OneShot_Distant_Creak_Att
 bEnableBusSends=false
 maxDistance=8200
 soundClassObject=&Fort_Ambient_OneShots
 volumeMultiplier=0.451
 preEffectBusSends [{}]
}
