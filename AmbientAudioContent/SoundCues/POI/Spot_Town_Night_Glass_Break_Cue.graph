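# Sound cue named for a spot night-time glass break one-shot in town POI ambience, but it has no root or nodes, so it references no sound to play.
# It keeps only playback settings: the ambient one-shots sound class, a reduced volume, and a tile-ambient bird one-shot attenuation, despite the glass-break name.
# Bus sends are disabled.
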SoundCue /AmbientAudioContent/SoundCues/POI/Spot_Town_Night_Glass_Break_Cue
root=null
nodes []
properties {
 attenuationSettings=&Fort_TileAmbient_OneShot_Bird_Att
 bEnableBusSends=false
 maxDistance=8005
 soundClassObject=&Fort_Ambient_OneShots
 volumeMultiplier=0.551
 preEffectBusSends [{}]
}
