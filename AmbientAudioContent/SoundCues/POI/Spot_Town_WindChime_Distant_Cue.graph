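# Sound cue for distant town wind chimes, an ambient one-shot at points of interest with distance attenuation.
# Two mixers each layer two random wave pools through pitch/volume modulators and envelopes that fade out by three seconds.
# The quality node sends its first two slots to the larger-pool mix and the last two to a reduced-pool mix.
# The reduced mix also swaps which layer, regular or distant chimes, is louder.
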SoundCue /AmbientAudioContent/SoundCues/POI/Spot_Town_WindChime_Distant_Cue
root=QualityLevel_0
nodes [
 {
  id=Enveloper_0
  children [Modulator_0]
  properties {
   pitchCurve {editorCurveData {keys [{time=0}]}}
   volumeCurve {editorCurveData {keys [<at 0 1 RCIM_Cubic>, {arriveTangent=0.0054, interpMode=RCIM_Cubic, leaveTangent=0.0054, tangentMode=RCTM_User, time=1.5, value=1}, <at 3 0>]}}
  }
 }
 {
  id=Enveloper_1
  children [Modulator_1]
  properties {
   pitchCurve {editorCurveData {keys [{time=0}]}}
   volumeCurve {editorCurveData {keys [
      {arriveTangent=0.0098, interpMode=RCIM_Cubic, leaveTangent=0.0098, tangentMode=RCTM_User, time=0, value=1}
      {arriveTangent=0.0054, interpMode=RCIM_Cubic, leaveTangent=0.0054, tangentMode=RCTM_User, time=1, value=1}
      <at 3 0>
     ]}}
  }
 }
 {
  id=Enveloper_4
  children [Modulator_5]
  properties {
   pitchCurve {editorCurveData {keys [{time=0}]}}
   volumeCurve {editorCurveData {keys [<at 0 1 RCIM_Cubic>, {arriveTangent=0.0054, interpMode=RCIM_Cubic, leaveTangent=0.0054, tangentMode=RCTM_User, time=1.5, value=1}, <at 3 0>]}}
  }
 }
 {
  id=Enveloper_5
  children [Modulator_4]
  properties {
   pitchCurve {editorCurveData {keys [{time=0}]}}
   volumeCurve {editorCurveData {keys [
      {arriveTangent=0.0098, interpMode=RCIM_Cubic, leaveTangent=0.0098, tangentMode=RCTM_User, time=0, value=1}
      {arriveTangent=0.0054, interpMode=RCIM_Cubic, leaveTangent=0.0054, tangentMode=RCTM_User, time=1, value=1}
      <at 3 0>
     ]}}
  }
 }
 {id=Mixer_0, children [Enveloper_0,Enveloper_1], properties {inputVolume [0.5,1]}}
 {id=Mixer_1, children [Enveloper_4,Enveloper_5], properties {inputVolume [1,1]}}
 {id=Modulator_0, children [Random_1], properties {pitchMax=1.1, pitchMin=1, volumeMax=0.6, volumeMin=0.5}}
 {id=Modulator_1, children [Random_2], properties {pitchMax=1.1, volumeMax=1.1, volumeMin=0.9}}
 {id=Modulator_4, children [Random_5], properties {pitchMax=1.1, volumeMax=1.1, volumeMin=0.9}}
 {id=Modulator_5, children [Random_3], properties {pitchMax=1.1, pitchMin=1, volumeMax=0.6, volumeMin=0.5}}
 {id=QualityLevel_0, children [Mixer_0,Mixer_0,Mixer_1,Mixer_1]}
 {
  id=Random_1
  children [
   WavePlayer_5
   WavePlayer_6
   WavePlayer_7
   WavePlayer_8
   WavePlayer_9
   WavePlayer_10
   WavePlayer_11
   WavePlayer_12
   WavePlayer_27
   WavePlayer_32
  ]
  properties {weights [1,1,1,1,1,1,1,1,1,1]}
 }
 {
  id=Random_2
  children [WavePlayer_33, WavePlayer_34, WavePlayer_35, WavePlayer_36, WavePlayer_37, WavePlayer_38]
  properties {weights [1,1,1,1,1,1]}
 }
 {id=Random_3, children [WavePlayer_14, WavePlayer_15, WavePlayer_16], properties {weights [1,1,1]}}
 {id=Random_5, children [WavePlayer_39, WavePlayer_40, WavePlayer_41, WavePlayer_42], properties {weights [1,1,1,1]}}
 {id=WavePlayer_10, properties {soundWaveAssetPtr=/AmbientAudioContent/Waves/OneShots/Spot_Town_Wind_Chimes_06}}
 {id=WavePlayer_11, properties {soundWaveAssetPtr=/AmbientAudioContent/Waves/OneShots/Spot_Town_Wind_Chimes_07}}
 {id=WavePlayer_12, properties {soundWaveAssetPtr=/AmbientAudioContent/Waves/OneShots/Spot_Town_Wind_Chimes_08}}
 {id=WavePlayer_14, properties {soundWaveAssetPtr=/AmbientAudioContent/Waves/OneShots/Spot_Town_WindChime_Distant_01}}
 {id=WavePlayer_15, properties {soundWaveAssetPtr=/AmbientAudioContent/Waves/OneShots/Spot_Town_WindChime_Distant_03}}
 {id=WavePlayer_16, properties {soundWaveAssetPtr=/AmbientAudioContent/Waves/OneShots/Spot_Town_WindChime_Distant_06}}
 {id=WavePlayer_27, properties {soundWaveAssetPtr=/AmbientAudioContent/Waves/OneShots/Spot_Town_Wind_Chimes_09}}
 {id=WavePlayer_32, properties {soundWaveAssetPtr=/AmbientAudioContent/Waves/OneShots/Spot_Town_Wind_Chimes_10}}
 {id=WavePlayer_33, properties {soundWaveAssetPtr=&Spot_Town_WindChime_Distant_01}}
 {id=WavePlayer_34, properties {soundWaveAssetPtr=/AmbientAudioContent/Waves/OneShots/Spot_Town_WindChime_Distant_02}}
 {id=WavePlayer_35, properties {soundWaveAssetPtr=&Spot_Town_WindChime_Distant_03}}
 {id=WavePlayer_36, properties {soundWaveAssetPtr=/AmbientAudioContent/Waves/OneShots/Spot_Town_WindChime_Distant_04}}
 {id=WavePlayer_37, properties {soundWaveAssetPtr=/AmbientAudioContent/Waves/OneShots/Spot_Town_WindChime_Distant_05}}
 {id=WavePlayer_38, properties {soundWaveAssetPtr=&Spot_Town_WindChime_Distant_06}}
 {id=WavePlayer_39, properties {soundWaveAssetPtr=&Spot_Town_Wind_Chimes_10}}
 {id=WavePlayer_40, properties {soundWaveAssetPtr=/AmbientAudioContent/Waves/OneShots/Spot_Town_Wind_Chimes_01}}
 {id=WavePlayer_41, properties {soundWaveAssetPtr=/AmbientAudioContent/Waves/OneShots/Spot_Town_Wind_Chimes_02}}
 {id=WavePlayer_42, properties {soundWaveAssetPtr=/AmbientAudioContent/Waves/OneShots/Spot_Town_Wind_Chimes_03}}
 {id=WavePlayer_5, properties {soundWaveAssetPtr=&Spot_Town_Wind_Chimes_01}}
 {id=WavePlayer_6, properties {soundWaveAssetPtr=&Spot_Town_Wind_Chimes_02}}
 {id=WavePlayer_7, properties {soundWaveAssetPtr=&Spot_Town_Wind_Chimes_03}}
 {id=WavePlayer_8, properties {soundWaveAssetPtr=/AmbientAudioContent/Waves/OneShots/Spot_Town_Wind_Chimes_04}}
 {id=WavePlayer_9, properties {soundWaveAssetPtr=/AmbientAudioContent/Waves/OneShots/Spot_Town_Wind_Chimes_05}}
]
properties {
 attenuationSettings=/Game/Sounds/Attenuation/Fort_TileAmbient_OneShot_Att
 bEnableBusSends=false
 maxDistance=8000
 soundClassObject=&Fort_Ambient_OneShots
 volumeMultiplier=0.4
 preEffectBusSends [{}]
}
