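# Sound cue for town point-of-interest ambient wind gust one-shots, named as the occluded interior variant.
# A Random node picks among five equally weighted gust waves and feeds a Modulator that varies pitch and boosts volume.
# All four quality-level inputs share that same branch.
# The cue lowers overall pitch and volume, uses the ambient one-shot sound class and wind attenuation, and disables bus sends.
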SoundCue /AmbientAudioContent/SoundCues/POI/Spot_Town_Wind_Gusts_Occluded_Int_Cue
root=QualityLevel_0
nodes [
 {id=Modulator_1, children [Random_2], properties {pitchMax=1.1, volumeMax=1.75, volumeMin=1.5}}
 {id=QualityLevel_0, children [Modulator_1,Modulator_1,Modulator_1,Modulator_1]}
 {id=Random_2, children [WavePlayer_8, WavePlayer_9, WavePlayer_10, WavePlayer_11, WavePlayer_12], properties {weights [1,1,1,1,1]}}
 {id=WavePlayer_10, properties {soundWaveAssetPtr=/AmbientAudioContent/Waves/OneShots/Spot_Town_Wind_Gust_03}}
 {id=WavePlayer_11, properties {soundWaveAssetPtr=/AmbientAudioContent/Waves/OneShots/Spot_Town_Wind_Gust_04}}
 {id=WavePlayer_12, properties {soundWaveAssetPtr=/AmbientAudioContent/Waves/OneShots/Spot_Town_Wind_Gust_05}}
 {id=WavePlayer_8, properties {soundWaveAssetPtr=/AmbientAudioContent/Waves/OneShots/Spot_Town_Wind_Gust_01}}
 {id=WavePlayer_9, properties {soundWaveAssetPtr=/AmbientAudioContent/Waves/OneShots/Spot_Town_Wind_Gust_02}}
]
properties {
 attenuationSettings=/Game/Sounds/Attenuation/Fort_TileAmbient_OneShot_Wind_Att
 bEnableBusSends=false
 maxDistance=8000
 pitchMultiplier=0.8
 soundClassObject=&Fort_Ambient_OneShots
 volumeMultiplier=0.125
 preEffectBusSends [{}]
}
