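# Ambient one-shot cue of distant warehouse metal creaks, impacts and light industrial hits, using the indoor ambient sound class and one-shot attenuation.
# The first two quality levels mix three randomized layers: delayed creaks, delayed impacts and undelayed light hits, with empty slots letting the latter two drop out.
# The last two levels play just one random sound from three.
# Modulators vary pitch and volume throughout, including over the full mix.
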SoundCue /AmbientAudioContent/SoundCues/POI/Spot_Warehouse_Metal_Clanking_01_Cue
root=QualityLevel_0
nodes [
 {id=Delay_12, children [Modulator_36], properties {delayMax=3}}
 {id=Delay_13, children [Modulator_37], properties {delayMax=2, delayMin=1}}
 {id=Mixer_19, children [Delay_12,Delay_13,Modulator_38], properties {inputVolume [2,1,0.5]}}
 {id=Modulator_1, children [WavePlayer_19], properties {pitchMax=1.1, pitchMin=0.6, volumeMax=1.5, volumeMin=0.6}}
 {id=Modulator_2, children [WavePlayer_18], properties {pitchMax=1.1, pitchMin=0.5, volumeMax=1, volumeMin=0.3}}
 {id=Modulator_3, children [WavePlayer_17], properties {pitchMax=1, pitchMin=0.5, volumeMax=0.5, volumeMin=0.2}}
 {id=Modulator_35, children [Mixer_19], properties {pitchMax=1.1, pitchMin=0.9, volumeMax=1.1, volumeMin=0.9}}
 {id=Modulator_36, children [Random_1], properties {pitchMax=1.1, pitchMin=0.9, volumeMax=1, volumeMin=0.3}}
 {id=Modulator_37, children [Random_5], properties {pitchMax=1.1, pitchMin=0.9, volumeMax=1, volumeMin=0.3}}
 {id=Modulator_38, children [Random_6], properties {pitchMax=1, pitchMin=0.5, volumeMax=1, volumeMin=0.3}}
 {id=QualityLevel_0, children [Modulator_35,Modulator_35,Random_4,Random_4]}
 {id=Random_1, children [WavePlayer_3,WavePlayer_4,WavePlayer_5,WavePlayer_6], properties {weights [1,1,1,1]}}
 {id=Random_4, children [Modulator_1,Modulator_2,Modulator_3], properties {weights [1,1,1]}}
 {id=Random_5, children [WavePlayer_7, WavePlayer_8, WavePlayer_9, WavePlayer_10, null], properties {weights [1,1,1,1,1]}}
 {id=Random_6, children [WavePlayer_20, WavePlayer_21, WavePlayer_22, null], properties {weights [1,1,1,1]}}
 {id=WavePlayer_10, properties {soundWaveAssetPtr=/AmbientAudioContent/Waves/OneShots/Spot_Industrial_Distant_Impact04}}
 {id=WavePlayer_17, properties {soundWaveAssetPtr=/AmbientAudioContent/Waves/OneShots/Spot_fort_industrial_light_05a}}
 {id=WavePlayer_18, properties {soundWaveAssetPtr=/AmbientAudioContent/Waves/OneShots/Spot_Industrial_Distant_Impact02}}
 {id=WavePlayer_19, properties {soundWaveAssetPtr=/AmbientAudioContent/Waves/OneShots/Spot_Industrial_Distant_Creak01}}
 {id=WavePlayer_20, properties {soundWaveAssetPtr=&Spot_fort_industrial_light_05a}}
 {id=WavePlayer_21, properties {soundWaveAssetPtr=/AmbientAudioContent/Waves/OneShots/Spot_fort_industrial_light_06a}}
 {id=WavePlayer_22, properties {soundWaveAssetPtr=/AmbientAudioContent/Waves/OneShots/Spot_fort_industrial_light_07a}}
 {id=WavePlayer_3, properties {soundWaveAssetPtr=&Spot_Industrial_Distant_Creak01}}
 {id=WavePlayer_4, properties {soundWaveAssetPtr=/AmbientAudioContent/Waves/OneShots/Spot_Industrial_Distant_Creak02}}
 {id=WavePlayer_5, properties {soundWaveAssetPtr=/AmbientAudioContent/Waves/OneShots/Spot_Industrial_Distant_Creak03}}
 {id=WavePlayer_6, properties {soundWaveAssetPtr=/AmbientAudioContent/Waves/OneShots/Spot_Industrial_Distant_Creak04}}
 {id=WavePlayer_7, properties {soundWaveAssetPtr=/AmbientAudioContent/Waves/OneShots/Spot_Industrial_Distant_Impact01}}
 {id=WavePlayer_8, properties {soundWaveAssetPtr=&Spot_Industrial_Distant_Impact02}}
 {id=WavePlayer_9, properties {soundWaveAssetPtr=/AmbientAudioContent/Waves/OneShots/Spot_Industrial_Distant_Impact03}}
]
properties {
 attenuationSettings=/Game/Sounds/Attenuation/Fort_TileAmbient_OneShot_Att
 maxDistance=8000
 soundClassObject=&Fort_Ambient_Indoors_Athena
 volumeMultiplier=0.501
}
