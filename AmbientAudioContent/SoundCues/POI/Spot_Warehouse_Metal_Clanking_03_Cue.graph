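# Indoor ambient one-shot cue of warehouse metal clanking, attenuated for distance, with a quality switch between layered and simpler versions.
# The first two quality levels mix delayed creaks, later distant impacts and undelayed light industrial hits, each randomly chosen with pitch and volume modulation.
# Null entries in those random pools let layers drop out for variation.
# The last two levels play just one of three modulated creak or industrial waves.
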SoundCue /AmbientAudioContent/SoundCues/POI/Spot_Warehouse_Metal_Clanking_03_Cue
root=QualityLevel_0
nodes [
 {id=Delay_12, children [Modulator_7], properties {delayMax=3}}
 {id=Delay_13, children [Modulator_4], properties {delayMax=4, delayMin=3}}
 {id=Mixer_19, children [Delay_12,Delay_13,Modulator_1], properties {inputVolume [2,1,0.5]}}
 {id=Modulator_1, children [Random_3], properties {pitchMax=1, pitchMin=0.5, volumeMax=1, volumeMin=0.3}}
 {id=Modulator_2, children [WavePlayer_14], properties {pitchMax=1, pitchMin=0.5, volumeMax=0.6, volumeMin=0.3}}
 {id=Modulator_35, children [Mixer_19], properties {pitchMax=1.1, pitchMin=0.9, volumeMax=1.1, volumeMin=0.9}}
 {id=Modulator_4, children [Random_2], properties {pitchMax=1.1, pitchMin=0.9, volumeMax=1, volumeMin=0.3}}
 {id=Modulator_5, children [WavePlayer_13], properties {pitchMax=1.1, pitchMin=0.5, volumeMax=1, volumeMin=0.3}}
 {id=Modulator_7, children [Random_0], properties {pitchMax=1.1, pitchMin=0.9, volumeMax=1, volumeMin=0.3}}
 {id=Modulator_8, children [WavePlayer_12], properties {pitchMax=1.1, pitchMin=0.5, volumeMax=1.5, volumeMin=0.6}}
 {id=QualityLevel_0, children [Modulator_35,Modulator_35,Random_4,Random_4]}
 {
  id=Random_0
  children [WavePlayer_1, WavePlayer_19, WavePlayer_20, WavePlayer_21, null]
  properties {bRandomizeWithoutReplacement=false, weights [1,1,1,1,0.5]}
 }
 {
  id=Random_2
  children [WavePlayer_34, WavePlayer_35, WavePlayer_36, WavePlayer_0, null]
  properties {bRandomizeWithoutReplacement=false, weights [1,1,1,1,0.5]}
 }
 {id=Random_3, children [WavePlayer_2, WavePlayer_26, WavePlayer_11, null], properties {bRandomizeWithoutReplacement=false, weights [1,1,1,1]}}
 {id=Random_4, children [Modulator_8,Modulator_5,Modulator_2], properties {weights [1,1,1]}}
 {id=WavePlayer_0, properties {soundWaveAssetPtr=/AmbientAudioContent/Waves/OneShots/Spot_Industrial_Distant_Impact17}}
 {id=WavePlayer_1, properties {soundWaveAssetPtr=/AmbientAudioContent/Waves/OneShots/Spot_Industrial_Distant_Creak09}}
 {id=WavePlayer_11, properties {soundWaveAssetPtr=/AmbientAudioContent/Waves/OneShots/Spot_fort_industrial_light_03a}}
 {id=WavePlayer_12, properties {soundWaveAssetPtr=&Spot_Industrial_Distant_Creak09}}
 {id=WavePlayer_13, properties {soundWaveAssetPtr=/AmbientAudioContent/Waves/OneShots/Spot_Industrial_Distant_Creak17}}
 {id=WavePlayer_14, properties {soundWaveAssetPtr=&Spot_fort_industrial_light_03a}}
 {id=WavePlayer_19, properties {soundWaveAssetPtr=/AmbientAudioContent/Waves/OneShots/Spot_Industrial_Distant_Creak10}}
 {id=WavePlayer_2, properties {soundWaveAssetPtr=/AmbientAudioContent/Waves/OneShots/Spot_fort_industrial_light_01a}}
 {id=WavePlayer_20, properties {soundWaveAssetPtr=/AmbientAudioContent/Waves/OneShots/Spot_Industrial_Distant_Creak11}}
 {id=WavePlayer_21, properties {soundWaveAssetPtr=/AmbientAudioContent/Waves/OneShots/Spot_Industrial_Distant_Creak12}}
 {id=WavePlayer_26, properties {soundWaveAssetPtr=/AmbientAudioContent/Waves/OneShots/Spot_fort_industrial_light_02a}}
 {id=WavePlayer_34, properties {soundWaveAssetPtr=/AmbientAudioContent/Waves/OneShots/Spot_Industrial_Distant_Impact18}}
 {id=WavePlayer_35, properties {soundWaveAssetPtr=/AmbientAudioContent/Waves/OneShots/Spot_Industrial_Distant_Impact19}}
 {id=WavePlayer_36, properties {soundWaveAssetPtr=/AmbientAudioContent/Waves/OneShots/Spot_Industrial_Distant_Impact20}}
]
properties {
 attenuationSettings=/Game/Sounds/Attenuation/Fort_TileAmbient_OneShot_Att
 maxDistance=8000
 soundClassObject=&Fort_Ambient_Indoors_Athena
 volumeMultiplier=0.501
}
