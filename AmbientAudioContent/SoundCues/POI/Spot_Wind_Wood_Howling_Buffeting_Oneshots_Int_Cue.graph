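# Sound cue for indoor ambient one-shots of howling, buffeting wind on wood, in the points-of-interest audio set.
# A Random node picks one of five equally weighted stereo waves and feeds a Modulator that varies volume on each play.
# It uses a wider tile-ambient one-shot attenuation and the indoor ambient sound class, with a concurrency override capping it at two simultaneous instances.
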SoundCue /AmbientAudioContent/SoundCues/POI/Spot_Wind_Wood_Howling_Buffeting_Oneshots_Int_Cue
root=Modulator_0
nodes [
 {id=Modulator_0, children [Random_0], properties {volumeMax=1.2}}
 {id=Random_0, children [WavePlayer_0,WavePlayer_1,WavePlayer_2,WavePlayer_3,WavePlayer_4], properties {weights [1,1,1,1,1]}}
 {id=WavePlayer_0, properties {soundWaveAssetPtr=/AmbientAudioContent/SoundCues/OneShots/Spot_Wind_Wood_Int_01_Stereo}}
 {id=WavePlayer_1, properties {soundWaveAssetPtr=/AmbientAudioContent/SoundCues/OneShots/Spot_Wind_Wood_Int_02_Stereo}}
 {id=WavePlayer_2, properties {soundWaveAssetPtr=/AmbientAudioContent/SoundCues/OneShots/Spot_Wind_Wood_Int_03_Stereo}}
 {id=WavePlayer_3, properties {soundWaveAssetPtr=/AmbientAudioContent/SoundCues/OneShots/Spot_Wind_Wood_Int_04_Stereo}}
 {id=WavePlayer_4, properties {soundWaveAssetPtr=/AmbientAudioContent/SoundCues/OneShots/Spot_Wind_Wood_Int_05_Stereo}}
]
properties {
 attenuationSettings=/Game/Sounds/Attenuation/Fort_TileAmbient_OneShot_Wider_Att
 bEnableBusSends=false
 bOverrideConcurrency=true
 maxDistance=8000
 soundClassObject=&Fort_Ambient_Indoors_Athena
 volumeMultiplier=1
 concurrencyOverrides {maxCount=2}
 preEffectBusSends [{}]
}
